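; Blueprint function library for NPC pawns that looks up and plays custom visual and sound effects by gameplay tag.
; Lookups go through the pawn's AICustomEffects component, and HasCustomEffect reports whether a visual asset is set.
; PlayCustomEffect runs only where cosmetic events can execute and in play mode, blocking-loading assets.
; It plays Niagara through the component and sound at location, using the named mesh socket or the actor transform.
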(fn GetCustomEffectByTag
  (params (CustomFXTag :type Gameplay%20Tag%20Structure) (PawnCustomVFXComp :type Fort%20Pawn%20Component%20AICustom%20Effects%20Object%20Reference))
  (if (Utilities|IsValid PawnCustomVFXComp :!class /Script/Engine.KismetSystemLibrary)
    (return (Class|FortPawnComponentAICustomEffects|GetCustomEffect PawnCustomVFXComp (Variables|GetCustomFXTag)))))

(fn GetCustomEffectLocation
  (params (SocketName :type Name) (Pawn :type Fort%20Pawn%20Object%20Reference))
  (if (!= SocketName "")
    (return (Transformation|GetSocketTransform (Variables|/Script/Engine.Character|GetMesh Pawn) SocketName "RTS_World"))
    (else
      (return (Transformation|GetActorTransform Pawn)))))

(fn HasCustomEffect
  (params (CustomFXTag :type Gameplay%20Tag%20Structure) (Pawn :type Fort%20Pawn%20Object%20Reference))
  (bind _output_get (LocalVariables|SetPawnCustomVFXComponent (Actor|GetComponentbyClass (Variables|GetPawn) "/Script/FortniteAI.FortPawnComponent_AICustomEffects")))
  (if (Utilities|IsValid _output_get :!class /Script/Engine.KismetSystemLibrary)
    (bind _customfx (UEFNCustomEffects|GetCustomEffectbyTag CustomFXTag (LocalVariables|GetPawnCustomVFXComponent) (LocalVariables|GetPawnCustomVFXComponent)))
    (bind _output_get_2 (LocalVariables|SetCustomEffect _customfx))
    (if (Asset|IsAssetNull (.VisualEffect (Utilities|Struct|BreakNPCCustomEffect _output_get_2)))
      (return false)
      (else
        (return true)))))

(fn PlayCustomEffect
  (params (CustomFXTag :type Gameplay%20Tag%20Structure) (Pawn :type Fort%20Pawn%20Object%20Reference))
  (bind _customeffect (LocalVariables|GetCustomEffect))
  (bind (_benableeffect _visualeffect _visualeffectsocket _soundeffect _soundeffectsocket) (Utilities|Struct|BreakNPCCustomEffect _customeffect))
  (Networking|CanExecuteCosmeticEvents
    (:True
      (if (Minigame|IsinPlayMode (Variables|GetPawn) :!class /Script/FortniteGame.MinigameBlueprintLibrary)
        ;[1 1 1] Play Visual Effect
        (bind _output_get (LocalVariables|SetPawnCustomVFXComponent (Actor|GetComponentbyClass (Variables|GetPawn) "/Script/FortniteAI.FortPawnComponent_AICustomEffects")))
        (if (Utilities|IsValid _output_get :!class /Script/Engine.KismetSystemLibrary)
          (bind _customfx (UEFNCustomEffects|GetCustomEffectbyTag CustomFXTag (LocalVariables|GetPawnCustomVFXComponent) (LocalVariables|GetPawnCustomVFXComponent)))
          (if (Utilities|Struct|BreakNPCCustomEffect _customfx)
            (LocalVariables|SetCustomEffect _customfx)
            (if (not (Asset|IsAssetNull (.VisualEffect (Utilities|Struct|BreakNPCCustomEffect _customfx))))
              (bind _effecttransformlocation (UEFNCustomEffects|GetCustomEffectLocation (.VisualEffectSocket (Utilities|Struct|BreakNPCCustomEffect _customfx)) (Variables|GetPawn) (Variables|GetPawn)))
              (bind _returnvalue (Utilities|LoadAssetBlocking (.VisualEffect (Utilities|Struct|BreakNPCCustomEffect _customfx))))
              (bind _asniagara_system (Utilities|Casting|CastToNiagaraSystem _returnvalue))
              (Class|FortPawnComponentAICustomEffects|PlayCustomEffect (LocalVariables|GetPawnCustomVFXComponent) _asniagara_system _effecttransformlocation))
            (if (Asset|IsAssetNull _soundeffect)
              (return true)
              (else
                (bind _effecttransformlocation_2 (UEFNCustomEffects|GetCustomEffectLocation _soundeffectsocket (Variables|GetPawn) (Variables|GetPawn)))
                (bind _returnvalue_2 (Utilities|LoadAssetBlocking _soundeffect))
                (bind _assound_base (Utilities|Casting|CastToSoundBase _returnvalue_2)
                  (:then
                    (Audio|PlaySoundatLocation 0 _assound_base (.location _effecttransformlocation_2) (.rotation _effecttransformlocation_2) 1.0 1.0 0.0 0 0 (Variables|GetPawn)))
                  (:CastFailed
                    _))
                (return true)))))))))
