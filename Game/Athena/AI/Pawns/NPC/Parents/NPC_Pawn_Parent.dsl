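; Parent pawn Blueprint for Athena NPCs, including wildlife, providing shared lifecycle, feedback and environment behavior.
; Begin play splits by authority: server sets up inventory, interaction and safe-zone polling; clients handle health bars, audio and invisible scale-in spawning.
; Alert-level changes drive interactability, sound and voice, with debug overrides.
; Gameplay effects, cues and tag-activated abilities handle spawning, conversion, swimming, drowning and storm damage, with many toggles read from hotfixable scalable floats.
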(event EventBeginPlay
  ;[1 1 1] Init / End
  ;;[1 1 1] Begin Play
  (Utilities|FlowControl|Sequence
    (:then_0
      (CallFunction|BindMovementModeChangeEvent)
      (UEFNCustomEffects|PlayCustomEffect "(TagName=\"NPC.CustomEffect.SpawnEffect\")" self) ; Play UEFN Customizable spawn effects
      (Class|FortActorComponentAffiliation|SetCanbeDamagedbyNullInstigator (Variables|Default|GetFortActorComponent_Affiliation) (Variables|NPCDamage|GetbCanBeDamagedByNullInstigator))
      (CallFunction|StoreOriginalGravityandRotation)
      (Variables|NPCMeshScaling|SetNPC%20OriginalMeshScale (Variables|/Script/Engine.SceneComponent|GetRelativeScale3D (Variables|Character|GetMesh))))
    (:then_1
      (Networking|SwitchHasAuthority
        (:Authority
          (CallFunction|InitializeInventory)
          (CallFunction|InitializeTestinSafeZone)
          (CallFunction|HandleInteracts)
          (Default|SelectPrimitiveDataFloat))
        (:Remote
          (CallFunction|StartUpdatingHleathBarWidget)
          (CallFunction|BeginMusicTracker)
          (CallFunction|AddSoundSystemLibraries)
          (CallFunction|NPCInitializeIdleVO)
          (Health|AssignOnPawnHealthChanged (EventDispatchers|CreateEvent :!event Custom|OnPawnHealthChanged_Event_0))
          (CallFunction|ResetMaterialfromDeath)
          (CallFunction|ForceUpdateCustomDepth) ; Ensure custom depth effects appear on pawns when they first spawn
          (Utilities|Time|SetTimerbyEvent (delegate Custom|UpdateCustomDepth) (Variables|Default|GetCustom%20Depth%20Update%20Rate) true)
          (if (not (Variables|NPCSpawning|GetNPC%20InvisiblePortionOfSpawningCompletedMakeVisible))
            (bind _performedsethidden (CallFunction|NPCSpawnInvisibleBeginsMakeInvisible)) ; Make pawn Invisible! (if we are allowed to)
            (if _performedsethidden
              (if (Variables|NPCSpawning|GetNPC%20ScaleMeshInOnSpawn)
                (Utilities|FlowControl|Delay 0.5) ; Small delay to minimize pops\nSome Spawn abilities do weird stuff and take a moment to set up
                (CallFunction|NPCScaleMeshIn)))))))))

(event Custom|ForceUpdateCustomDepth
  (bind _fortpawncomponent_customdepth (Variables|Default|GetFortPawnComponent_CustomDepth))
  ;[1 1 1] Init / End
  ;;[1 1 1] Begin Play
  (if (CustomDepth|IsRenderCustomDepthEnabled _fortpawncomponent_customdepth)
    (CustomDepth|Update _fortpawncomponent_customdepth 0.0 true)))

(event Custom|Debug_AlertLevel_ForceLevel
  (params (NewNPC_AlertLevel :type Enum_NPC_AlertLevel%20Enum :default "NewEnumerator4"))
  ;[1 0 0.6403] DEBUG COMMANDS
  ;;[1 1 1] Debug Alert Level Commands
  (bind _output_get (Variables|NPCAlertLevel|SetNPC_AlertLevelPrevious NewNPC_AlertLevel))
  (Variables|NPCAlertLevel|SetNPC_AlertLevel _output_get)
  (Variables|NPCAlertLevel|SetDebug_NPC_AlertLevelForced true))

(event Custom|End%20Music%20Tracker
  ;[0.4219 1 0] Audio
  ;;[1 1 1] Music Tracker
  (bind _returnvalue (FortGameplayDataTrackerComponent|GetFortGameplayDataTrackerComponentManager))
  (Utilities|IsValid _returnvalue
    (:Is%20Valid
      (Class|FortGameplayDataTrackerComponentManager|UnRegisterTrackedStateActor _returnvalue self))))

(event Custom|NPC%20LeaderSet
  (params (Success? :type Boolean) (Leader :type Actor%20Object%20Reference) (UnconvertReason :type EUnconvertReason%20Enum :default "StealByOtherPlayer"))
  (bind _npc_skeletal_mesh_index (Variables|Internal|GetNPC%20Skeletal%20Mesh%20Index))
  (bind _returnvalue (Ability|GameplayCue|MakeGameplayCueParameters :NormalizedMagnitude 0.0 :RawMagnitude _npc_skeletal_mesh_index :Location "0,0,0" :Normal "0,0,0" :Instigator Leader :GameplayEffectLevel 1 :AbilityLevel 1 :bReplicateLocationWhenUsingMinimalRepProxy false))
  ;[0 1 0.8831] General Logic
  ;;[1 1 1] Leader has been set. Play any VFX or logic we need
  (Networking|SwitchHasAuthority
    (:Authority
      (if Success?
        (bind _ge_converted (Variables|NPCLeader|GetGE_Converted)
          (:then
            (bind _aipawnabilitysystemcomponent (Variables|GameplayAbilities|GetAIPawnAbilitySystemComponent)
              (:then
                (bind _returnvalue_2 (GameplayEffects|ApplyGameplayEffectToSelf :self _aipawnabilitysystemcomponent :GameplayEffectClass _ge_converted :Level 0.0)) ; Grant tag to give new goal selection params
                (Variables|NPCLeader|SetGE_Converted_Handle _returnvalue_2))
              (:else)))
          (:else))
        (else
          (bind _aipawnabilitysystemcomponent_2 (Variables|GameplayAbilities|GetAIPawnAbilitySystemComponent)
            (:then
              (GameplayEffects|RemoveActiveGameplayEffect _aipawnabilitysystemcomponent_2 (Variables|NPCLeader|GetGE_Converted_Handle) -1))
            (:else)))))
    (:Remote
      (if Success?
        (if (not (and (GameplayTags|IsGameplayTagValid (Variables|NPCLeader|GetGameplayCue_LeaderSet)) (Variables|NPCLeader|GetisGCForLeaderSetLooping)))
          (GameplayCue|ExecuteGameplayCue%28Burst%29Local self (Variables|NPCLeader|GetGameplayCue_LeaderSet) _returnvalue))
        (bind _tamedsound (Variables|Sound|GetTamedSound)
          (:then
            (Audio|PlaySoundatLocation _tamedsound (Transformation|GetActorLocation)))
          (:else))
        (else
          (switch Utilities|FlowControl|Switch|SwitchonEUnconvertReason UnconvertReason
            (:ConverterDeath
              _)
            (:TooManyConverted
              _)
            (:TooFarFromConverter
              _))
          (if (GameplayTags|IsGameplayTagValid (Variables|NPCLeader|GetGameplayCue_LeaderCleared))
            (Utilities|FlowControl|Delay (Variables|NPCLeader|GetLeaderClearedCueDelay)) ; Short delay to let leader set cue ↑ play first to improve readability
            (GameplayCue|ExecuteGameplayCue%28Burst%29Local self (Variables|NPCLeader|GetGameplayCue_LeaderCleared) _returnvalue)))))))

(event Custom|Debug_AlertLevel_ClearDebug
  ;[1 0 0.6403] DEBUG COMMANDS
  ;;[1 1 1] Debug Alert Level Commands
  (bind _output_get (Variables|NPCAlertLevel|SetNPC_AlertLevelPrevious "NewEnumerator4"))
  (Variables|NPCAlertLevel|SetNPC_AlertLevel _output_get)
  (Variables|NPCAlertLevel|SetDebug_NPC_AlertLevelForced false))

(event Custom|BeginMusicTracker
  (bind (_alertlevel _alertleveltag) (CallFunction|NPCGetAlertLevelInfo))
  ;[0.4219 1 0] Audio
  ;;[1 1 1] Music Tracker
  (bind _returnvalue (FortGameplayDataTrackerComponent|GetFortGameplayDataTrackerComponentManager))
  (Utilities|IsValid _returnvalue
    (:Is%20Valid
      (Class|FortGameplayDataTrackerComponentManager|RegisterTrackedStateActor _returnvalue self _alertleveltag))))

(event Custom|Debug_ForceAlertLevel_Invalid
  ;[1 0 0.6403] DEBUG COMMANDS
  ;;[1 1 1] Debug Alert Level Commands
  (CallFunction|DebugAlertLevelForceLevel "NewEnumerator4"))

(event Custom|Debug_ForceAlertLevel_NA
  ;[1 0 0.6403] DEBUG COMMANDS
  ;;[1 1 1] Debug Alert Level Commands
  (CallFunction|DebugAlertLevelForceLevel "NewEnumerator0"))

(event Custom|Debug_ForceAlertLevel_Unaware
  ;[1 0 0.6403] DEBUG COMMANDS
  ;;[1 1 1] Debug Alert Level Commands
  (CallFunction|DebugAlertLevelForceLevel "NewEnumerator1"))

(event Custom|UpdateCustomDepth
  (bind _fortpawncomponent_customdepth (Variables|Default|GetFortPawnComponent_CustomDepth))
  ;[1 1 1] Init / End
  ;;[1 1 1] Begin Play
  (if (CustomDepth|IsRenderCustomDepthEnabled _fortpawncomponent_customdepth)
    (CustomDepth|Update _fortpawncomponent_customdepth (Variables|Default|GetCustom%20Depth%20Update%20Rate) false)))

(event Custom|TestLoot
  ;[1 0 0.6403] DEBUG COMMANDS
  ;;[1 1 1] Loot Debug
  (Variables|NPCInventory|SetNPC%20BestLootRarityInInventory "Legendary"))

(event Custom|HandleInteracts
  ;[0 1 0.8831] General Logic
  ;;[1 1 1] EventHandleInteracts-> Interacts defaulted to false.  Mostly for use with wildlife children.
  (bind _asbool (CallFunction|NPCGetScalableFloatHotfixValue (Variables|NPCInteraction|GetNPC%20PawnCanInteractHF)))
  (Variables|NPCInteraction|SetNPC%20PawnCanInteractInternal _asbool)
  (bind _caninteract (CallFunction|NPCCanbeInteractedWith))
  (bind _enabledinteract (CallFunction|NPCUpdateInteract _caninteract))
  (bind _output_get (Variables|NPCInteraction|SetNPCInteractEnabled _enabledinteract))
  (if _output_get
    (AI|AssignOnInteraction (delegate Custom|NPC%20Pawn%20Interact))
    (else
      (AI|UnbindEventfromOnInteraction (delegate Custom|NPC%20Pawn%20Interact)))))

(event Custom|UpdateSoundLibFoley
  ;[0.4219 1 0] Audio
  ;;[1 1 1] Sound System Libraries
  (CallFunction|GetSoundFoleyLibrary))

(event Custom|AddSoundSystemLibraries
  ;[0.4219 1 0] Audio
  ;;[1 1 1] Sound System Libraries
  (bind _validlibraries (SoundLibraries|UpdateSoundLibraryCollection true))
  (if (not _validlibraries)
    (SoundLibrary|AddLibraries (WorldSubsystems|GetSoundLibrarySubsystem) self (Variables|SoundLibraries|GetSoundLibs_Default) :!class /Script/SoundLibrary.SoundLibrarySubsystem))
  ;;;[1 1 1] Check to see if water index is valid
  (if (Utilities|Array|IsValidIndex (Variables|SoundLibraries|GetSoundLibs_Foley_Water))
    (CallFunction|UpdateSoundLibFoley)
    (else
      (SoundLibrary|AddLibraries (WorldSubsystems|GetSoundLibrarySubsystem) self (Variables|SoundLibraries|GetSoundLibs_Foley) :!class /Script/SoundLibrary.SoundLibrarySubsystem))))

(event Custom|Debug_ForceAlertLevel_Alerted
  ;[1 0 0.6403] DEBUG COMMANDS
  ;;[1 1 1] Debug Alert Level Commands
  (CallFunction|DebugAlertLevelForceLevel "NewEnumerator2"))

(event Custom|Debug_ForceAlertLevel_Aggressive
  ;[1 0 0.6403] DEBUG COMMANDS
  ;;[1 1 1] Debug Alert Level Commands
  (CallFunction|DebugAlertLevelForceLevel "NewEnumerator3"))

(event Custom|NPC%20Pawn%20Interact
  (params (AI :type Fort%20AIPawn%20Object%20Reference) (InteractingPlayerController :type Fort%20Player%20Controller%20Object%20Reference) (InteractionBeingAttempted :type EInteractionBeingAttempted%20Enum :default "FirstInteraction"))
  ;[0 1 0.8831] General Logic
  ;;[1 1 1] EventHandleInteracts-> Interacts defaulted to false.  Mostly for use with wildlife children.
  (CallFunction|NPCUpdateInteract false)
  (CallFunction|OnNPCPawnInteractedWith AI (Game|Player|GetPlayerPawn InteractingPlayerController) InteractionBeingAttempted))

(event Custom|RemoveSoundSystemLibraries
  (bind _returnvalue (WorldSubsystems|GetSoundLibrarySubsystem))
  (bind _self self)
  (bind _soundlibs_foley_water (Variables|SoundLibraries|GetSoundLibs_Foley_Water))
  ;[0.4219 1 0] Audio
  ;;[1 1 1] Sound System Libraries
  (bind _validlibraries (SoundLibraries|UpdateSoundLibraryCollection false))
  (if (not _validlibraries)
    (SoundLibrary|RemoveLibraries (WorldSubsystems|GetSoundLibrarySubsystem) self (Variables|SoundLibraries|GetSoundLibs_Default) :!class /Script/SoundLibrary.SoundLibrarySubsystem))
  (SoundLibrary|RemoveLibraries _returnvalue _self (Variables|SoundLibraries|GetSoundLibs_Foley) :!class /Script/SoundLibrary.SoundLibrarySubsystem)
  (if (Utilities|Array|IsValidIndex _soundlibs_foley_water)
    (SoundLibrary|RemoveLibraries _returnvalue _self _soundlibs_foley_water :!class /Script/SoundLibrary.SoundLibrarySubsystem)))

(event Custom|On%20NPC%20PawnInteractedWith
  (params (AIPawn :type Actor%20Object%20Reference) (InteractingPlayerPawn :type Fort%20Player%20Pawn%20Object%20Reference) (InteractionBeingAttempted :type EInteractionBeingAttempted%20Enum :default "FirstInteraction"))
  ;[0 1 0.8831] General Logic
  ;;[1 1 1] EventHandleInteracts-> Interacts defaulted to false.  Mostly for use with wildlife children.
  (Default|CallNPCPawnInteractwithDispatcher AIPawn InteractingPlayerPawn InteractionBeingAttempted))

(event EventPossessed
  (params (NewController :type Controller%20Object%20Reference))
  (bind _tc_npc_alternatespawnabilitytag (Variables|NPCSpawning|GetTC_NPC%20AlternateSpawnAbilityTag))
  ;[1 1 1] Init / End
  ;;[1 1 1] On Possess
  (if (GameplayTags|NotEqual(GameplayTagContainer) (Variables|NPCSpawning|GetTC_NPC%20SpawnAbilityTag) (GameplayTags|MakeLiteralGameplayTagContainer))
    (Utilities|FlowControl|Delay 0.0001) ; Tiny delay otherwise the ability anims may pop
    (if (GameplayTags|NotEqual(GameplayTagContainer) (Variables|NPCSpawning|GetTC_NPC%20AlternateSpawnAbilityTag) (GameplayTags|MakeLiteralGameplayTagContainer))
      (bind _aipawnabilitysystemcomponent (Variables|GameplayAbilities|GetAIPawnAbilitySystemComponent)
        (:then
          (if (Abilities|CanActivateAbilitywithMatchingTag _aipawnabilitysystemcomponent _tc_npc_alternatespawnabilitytag)
            (Abilities|TryActivateAbilitiesbyTag _aipawnabilitysystemcomponent (GameplayTags|MakeLiteralGameplayTagContainer _tc_npc_alternatespawnabilitytag) false) ; Activate Alternate Spawn
            (else
              _)))
        (:else)))
    (bind _aipawnabilitysystemcomponent_2 (Variables|GameplayAbilities|GetAIPawnAbilitySystemComponent)
      (:then
        (Abilities|TryActivateAbilitiesbyTag _aipawnabilitysystemcomponent_2 (Variables|NPCSpawning|GetTC_NPC%20SpawnAbilityTag) false) ; Activate Primary Spawn
      )
      (:else))))

(event Custom|NPCEnableHealthBars
  ;[1 0 0.6403] DEBUG COMMANDS
  ;;[1 1 1] Mostly for capture team stuff to hide bars across the board.
  (Variables|NPCStatusWidgetHealthBar|SetNPC%20StatusWidgetHealthBarDisabled false))

(event Custom|NPC%20InitializeIdleVO
  ;[0.4219 1 0] Audio
  ;;[1 1 1] Handle IDLE VO
  (Utilities|Time|SetTimerbyEvent (delegate Custom|TriggerIdleVO) (Math|Random|RandomFloatinRange (ScalableFloat|GetValueatLevel (Variables|VoiceComponent|GetMinIdleVODelay)) (ScalableFloat|GetValueatLevel (Variables|VoiceComponent|GetMaxIdleVODelay))) true))

(event Custom|NPCDisableHealthBars
  ;[1 0 0.6403] DEBUG COMMANDS
  ;;[1 1 1] Mostly for capture team stuff to hide bars across the board.
  (Variables|NPCStatusWidgetHealthBar|SetNPC%20StatusWidgetHealthBarDisabled true)
  (bind _npcstatuswidget (Variables|Default|GetNPCStatusWidget)
    (:then
      (bind _asnpcstatus_widget (Utilities|Casting|CastToNPCStatusWidget (UserInterface|GetUserWidgetObject _npcstatuswidget)))
      (Widget|SetVisibility _asnpcstatus_widget "Hidden"))
    (:else)))

(event Custom|InitializeInventory
  ;[0 1 0.8831] General Logic
  ;;[1 1 1] Initialize Inventory
  (FortItem|BindEventtoOnInventoryInitialized (delegate Custom|GrantStartingItems)))

(event Custom|TriggerIdleVO
  ;[0.4219 1 0] Audio
  ;;[1 1 1] Handle IDLE VO
  (Class|NPCVoiceComponent|PlayVoiceLine (Variables|Default|GetNPC_VoiceComponent) (select (Utilities|Enum|Equal(Enum) (Variables|NPCAlertLevel|GetNPC_AlertLevel) "NewEnumerator3") (Variables|VoiceComponent|GetVoiceTag_Idle) (Variables|VoiceComponent|GetVoiceTag_FullAlertIdle)) (Variables|Default|GetVOAudioComponent)))

(event Custom|GrantStartingItems
  ;[0 1 0.8831] General Logic
  ;;[1 1 1] Initialize Inventory
  (FortItem|UnbindEventfromOnInventoryInitialized (delegate Custom|GrantStartingItems))
  (Variables|NPCInventory|SetNPC_IsInventoryInitialized true)
  (bind _aipawnabilitysystemcomponent (Variables|GameplayAbilities|GetAIPawnAbilitySystemComponent)
    (:then
      (Abilities|TryActivateAbilitiesbyTag _aipawnabilitysystemcomponent (Variables|NPCInventory|GetGrantStartingItemsAbilityTC) false)
      (Utilities|FlowControl|Delay (Variables|NPCInventory|GetInventoryGlowDelay))
      (CallFunction|NPCSetBestLootRarityinInventory) ; currently crashes when doing this quickly. async on the mesh not being loaded?\n\nsets what our best inventory item is, and may glow that color as a result of this function
    )
    (:else)))

(event EventEndPlay
  (params (EndPlayReason :type EEndPlayReason%20Enum :default "Destroyed"))
  ;[1 1 1] Init / End
  ;;[1 1 1] End Play (Cleanup)
  (CallFunction|NPCClearOutTargetSlots)
  (Networking|SwitchHasAuthority
    (:Remote
      (CallFunction|StopUpdatingHealthBarWidget)
      (CallFunction|EndMusicTracker)
      (CallFunction|RemoveSoundSystemLibraries))))

(event Custom|NPCAlertLevelChangedClient
  (params (NPC_AlertLevel :type Enum_NPC_AlertLevel%20Enum :default "NewEnumerator4"))
  ;[0.4219 1 0] Audio
  ;;[1 1 1] Update Sound Library when Alert Level Changed (and lured for VFX)
  (bind _validlibraries (SoundLibraries|UpdateSoundLibraryCollection true))
  (if (not _validlibraries)
    (Sound|UpdateSoundLibrary NPC_AlertLevel))
  (CallFunction|UpdateTrackedActorStatefromCurrentAlertLevel)
  (CallFunction|NPCHandleAlertLevelVoice)
  (Audio|Parameter|SetIntegerParameter (Variables|Default|GetMovementAudioComponent) "IsAggressive" (Math|Conversions|ToInteger(Boolean) (Utilities|Enum|Equal(Enum) (Variables|NPCAlertLevel|GetNPC_AlertLevel) "NewEnumerator3")) :!class /Script/AudioExtensions.AudioParameterControllerInterface))

(event Custom|RequestUpdateNPCAlertLevel
  (params (NPC_AlertLevel :type Enum_NPC_AlertLevel%20Enum :default "NewEnumerator4") (Reason :type Enum_NPC_ModifyAlertLevelReason%20Enum :default "NewEnumerator4"))
  ;[0 1 0.8831] General Logic
  (CallFunction|LuredonServer NPC_AlertLevel) ; TEMP - Delete when we can bind to goal selection changed
  ;;[1 1 1] Update Alert Level (if changed)
  (if (Variables|NPCAlertLevel|GetDebug_NPC_AlertLevelForced)
    _
    (elif (Utilities|Enum|NotEqual(Enum) NPC_AlertLevel (Variables|NPCAlertLevel|GetNPC_AlertLevel))
      (Variables|NPCAlertLevel|SetNPC_AlertLevelPrevious (Variables|NPCAlertLevel|GetNPC_AlertLevel))
      (bind _output_get (Variables|NPCAlertLevel|SetNPC_AlertLevel NPC_AlertLevel))
      (CallFunction|NPCAlertLevelChangedServer _output_get))))

(event Custom|NPC%20AlertLevelChangedServer
  (params (NPCAlertLevel :type Enum_NPC_AlertLevel%20Enum :default "NewEnumerator4"))
  ;[0 1 0.8831] General Logic
  ;;[1 1 1] Disable Interact When Alerted or Aggressive
  (switch Utilities|FlowControl|Switch|SwitchonEnum_NPC_AlertLevel NPCAlertLevel
    (:NewEnumerator1
      (Ability|SendGameplayEventtoActor self (Variables|NPCIntroPhase|GetNPC%20ResetIntroGameplayTags) (Utilities|Struct|MakeGameplayEventData (Variables|NPCIntroPhase|GetNPC%20ResetIntroGameplayTags)))
      (bind _asbool (CallFunction|NPCGetScalableFloatHotfixValue (Variables|NPCInteraction|GetNPC%20AllowInteractOnlyWhenUnawareHF)))
      (if _asbool
        (bind _caninteract (CallFunction|NPCCanbeInteractedWith))
        (CallFunction|NPCUpdateInteract _caninteract)))
    (:NewEnumerator2
      _)
    (:NewEnumerator3
      _))
  (if (not (Default|IsConverted))
    (bind _asbool_2 (CallFunction|NPCGetScalableFloatHotfixValue (Variables|NPCInteraction|GetNPC%20AllowInteractOnlyWhenUnawareHF)))
    (if _asbool_2
      (CallFunction|NPCUpdateInteract false))))

(event Custom|StartUpdatingHleathBarWidget
  ;[0.3127 0 1] UI
  ;;[1 1 1] Listen to widget visibility changes and disable tick when hidden
  (bind _npcstatuswidget (Variables|Default|GetNPCStatusWidget)
    (:then
      (bind _asnpcstatus_widget (Utilities|Casting|CastToNPCStatusWidget (UserInterface|GetUserWidgetObject _npcstatuswidget)))
      (Utilities|FlowControl|Sequence
        (:then_0
          (Components|Tick|SetComponentTickEnabled (Variables|Default|GetNPCStatusWidget) (Widget|IsVisible _asnpcstatus_widget)))
        (:then_1
          (Default|BindEventtoVisibilityChanged _asnpcstatus_widget (delegate Custom|WidgetVisibilityChanged) :!class /Game/Athena/HUD/Actors/NPCStatusWidget.NPCStatusWidget_C))))
    (:else)))

(event Custom|WidgetVisibilityChanged
  (params (bVisible :type Boolean))
  ;[0.3127 0 1] UI
  ;;[1 1 1] Listen to widget visibility changes and disable tick when hidden
  (Components|Tick|SetComponentTickEnabled (Variables|Default|GetNPCStatusWidget) bVisible))

(event Custom|StopUpdatingHealthBarWidget
  ;[0.3127 0 1] UI
  ;;[1 1 1] Listen to widget visibility changes and disable tick when hidden
  (bind _npcstatuswidget (Variables|Default|GetNPCStatusWidget)
    (:then
      (bind _asnpcstatus_widget (Utilities|Casting|CastToNPCStatusWidget (UserInterface|GetUserWidgetObject _npcstatuswidget)))
      (Default|UnbindEventfromVisibilityChanged _asnpcstatus_widget (delegate Custom|WidgetVisibilityChanged) :!class /Game/Athena/HUD/Actors/NPCStatusWidget.NPCStatusWidget_C))
    (:else)))

(event EventOnDamageServer
  (params (Damage :type Float%20%28single-precision%29) (DamageTags :type Gameplay%20Tag%20Container%20Structure%20%28by%20ref%29) (Momentum :type Vector :default "0, 0, 0") (HitInfo :type Hit%20Result%20Structure%20%28by%20ref%29) (InstigatedBy :type Controller%20Object%20Reference) (DamageCauser :type Actor%20Object%20Reference) (EffectContext :type Gameplay%20Effect%20Context%20Handle%20Structure))
  ;[1 0.4746 0] Damage Handling
  ;;[1 1 1] On Damage Server
  (Utilities|FlowControl|Sequence
    (:then_0
      (CallFunction|NPCApplyRecentlyDamagedTagGE 0) ; Apply GE/tags that say "I have taken damage recently."\n\nBy default these also disable the GEs with NPC.Behavior.Phase.Idle
    )
    (:then_1
      (CallFunction|NPCDamageAttemptDisableIntroPhaseBehavior DamageTags) ; If we have an Intro Phase Behavior (EX: roaring, circling, telegraphing) should we cancel it \nand move straight into combat or avoidance behaviors?
    )
    (:then_2
      (bind _asbool (CallFunction|NPCGetScalableFloatHotfixValue (Variables|NPCDespawnfromStorm|GetNPC%20ShouldDespawnFromStormHF)))
      (if (and _asbool (Variables|NPCDespawnfromStorm|GetNPC%20isTryingToDespawn))
        (CallFunction|TrytoDespawnDuetoStorm true)))))

(event Custom|NPCScaleMeshIn
  ;[1 1 1] Mesh Scale In and Out for Spawn/Despawn
  (if (not (Variables|NPCMeshScaling|GetNPC%20MeshScalingInProgress))
    (Variables|NPCMeshScaling|SetNPC%20MeshScalingInProgress true)
    (Components|Timeline|SetPlayRate (Variables|NPC_Pawn_Parent|GetTimeline_ScaleMeshInOrOut) (Variables|NPCSpawning|GetScaleMeshInOrOutPlayRate))
    (bind (_direction _meshscalealpha) (Timeline|Timeline_ScaleMeshInOrOut PlayFromStart)
      (:Update
        (Transformation|SetRelativeScale3D (Variables|Character|GetMesh) (Math|Vector|Lerp(Vector) "0.001000,0.001000,0.001000" (Variables|NPCMeshScaling|GetNPC%20OriginalMeshScale) _meshscalealpha))
        (switch Utilities|FlowControl|Switch|SwitchonETimelineDirection _direction
          (:Forward
            (if (not (Variables|NPCSpawning|GetNPC%20InvisiblePortionOfSpawningCompletedMakeVisible))
              (Variables|NPCSpawning|SetNPC%20InvisiblePortionOfSpawningCompletedMakeVisible true) ; Make Visible again!
            ))))
      (:Finished
        (Variables|NPCMeshScaling|SetNPC%20MeshScalingInProgress false)
        (switch Utilities|FlowControl|Switch|SwitchonETimelineDirection _direction
          (:Forward
            (Default|CallNPCScaleMeshinCompleted)
            (Transformation|SetRelativeScale3D (Variables|Character|GetMesh) (Variables|NPCMeshScaling|GetNPC%20OriginalMeshScale)) ; Make SURE we get back to our original Scale
          )
          (:Backward
            (Default|CallNPCScaleMeshOutCompleted)))))))

(event Custom|NPC%20UpdateHealthBarComponentVisibility
  (bind _returnvalue (Utilities|Time|GetGameTimeinSeconds))
  (bind _npc_death_time (Variables|Default|GetNPC%20Death%20Time))
  (bind _returnvalue_2 (- _returnvalue _npc_death_time))
  ;[0.3127 0 1] UI
  ;;[1 1 1] Health Bar Handling on Death
  (if (> (Variables|Default|GetNPC%20Death%20Time) 0)
    (bind (_asbool _asintrounded _asfloat) (CallFunction|NPCGetScalableFloatHotfixValue (Variables|NPCStatusWidgetHealthBar|GetNPC%20HealthBarHideTimeAfterDeathHF)))
    (if (and (> _asfloat 0.0) (not (Variables|NPCStatusWidgetHealthBar|GetNPC%20StatusWidgetHealthBarDisabled)))
      (bind _npcstatuswidget (Variables|Default|GetNPCStatusWidget)
        (:then
          (bind _asnpcstatus_widget (Utilities|Casting|CastToNPCStatusWidget (UserInterface|GetUserWidgetObject _npcstatuswidget)))
          (Class|NPCStatusWidget|EventUpdateStatus _asnpcstatus_widget 0.0 0.0 false :!pinned true) ; Make sure  Health Bar is 0
          (bind (_asbool_2 _asintrounded_2 _asfloat_2) (CallFunction|NPCGetScalableFloatHotfixValue (Variables|NPCStatusWidgetHealthBar|GetNPC%20HealthBarHideTimeAfterDeathHF)))
          (if (> (- _asfloat_2 _returnvalue_2) 0)
            (Utilities|FlowControl|Delay (- _asfloat_2 _returnvalue_2))))
        (:else)))
    (bind _npcstatuswidget_2 (Variables|Default|GetNPCStatusWidget)
      (:then
        (Development|SetHiddeninGame _npcstatuswidget_2 true false :!pinned true) ; Hide health bar (probably after animation)
      )
      (:else))))

(event Custom|NPC%20ParentLaunchVehicle
  (params (Vehicle :type Actor%20Object%20Reference) (Velocity :type Vector))
  ;[1 0 0] Combat Related
  ;;[1 1 1] This is how we launch vehicles. Call from ability damage.
  ;; No, it doesn't work in the ability itself, without being Replicated. :[
  (Audio|BroadcastAbilityImpactatLocation 0 Vehicle Velocity 777.0 "None" true))

(event Custom|NPCScaleMeshOut
  ;[1 1 1] Mesh Scale In and Out for Spawn/Despawn
  (Utilities|FlowControl|Sequence
    (:then_0
      (if (Variables|NPCMeshScaling|GetNPC%20DisableCapsuleOverlapsWhenScalingOut)
        (bind _capsulecomponent (Variables|Character|GetCapsuleComponent)
          (:then
            (Variables|/Script/Engine.PrimitiveComponent|SetbGenerateOverlapEvents false _capsulecomponent))
          (:else))))
    (:then_1
      (if (not (Variables|NPCMeshScaling|GetNPC%20MeshScalingInProgress))
        (Networking|SwitchHasAuthority
          (:Authority
            _)
          (:Remote
            (if (and (Variables|NPCSpawning|GetAllowSkipScaleOutIfNotRecentlyRendered) (not (Rendering|WasActorRecentlyRendered self 0.2)))
              (Rendering|SetActorHiddenInGame true)
              (else
                _))))
        (Variables|NPCMeshScaling|SetNPC%20MeshScalingInProgress true)
        (Components|Timeline|SetPlayRate (Variables|NPC_Pawn_Parent|GetTimeline_ScaleMeshInOrOut) (Variables|NPCSpawning|GetScaleMeshInOrOutPlayRate))
        (Timeline|Timeline_ScaleMeshInOrOut ReverseFromEnd)))))

(event Custom|OnPawnHealthChanged_Event_0
  ;[0.3127 0 1] UI
  ;;[1 1 1] On Pawn Health Changed - Update Health Bar
  (CallFunction|TryShowHealthBar))

(event WaterBody|EventOnBeginDeepWaterInteraction
  (params (WaterSurfaceLocation :type Vector :default "0, 0, 0") (WaterDepth :type Float%20%28single-precision%29) (WaterBody :type Water%20Body%20Object%20Reference))
  ;[1 0.1868 0] Despawn / Death Handling
  ;;[1 1 1] Deep Water Interaction
  (Utilities|FlowControl|Sequence
    (:then_0
      (if (Variables|NPCSwimming|GetNPC%20Apply%20Deep%20Water%20Enter/Exit%20Burst%20GC)
        (bind _asfort_water_body_actor (Utilities|Casting|CastToFortWaterBodyActor WaterBody))
        (CallFunction|PlayWaterSplashAtWaterSurface _asfort_water_body_actor (Variables|NPCSwimming|GetEnter%20Deep%20Water))))
    (:then_1
      (bind _asbool (CallFunction|NPCGetScalableFloatHotfixValue (Variables|NPCSwimming|GetNPC%20AllowDrownInShallowWaterHF)))
      (if (not _asbool)
        (CallFunction|TryWaterInteractionDrown true)))))

(event Custom|BindMovementModeChangeEvent
  ;[1 0.1868 0] Despawn / Death Handling
  (Networking|SwitchHasAuthority
    (:Authority
      (Character|AssignMovementModeChangedDelegate (delegate Custom|MovementModeChangedDelegate_Event)))))

(event Custom|MovementModeChangedDelegate_Event
  (params (Character :type Character%20Object%20Reference) (PrevMovementMode :type EMovementMode%20Enum :default "MOVE_None") (PreviousCustomMode :type Byte))
  ;[1 0.1868 0] Despawn / Death Handling
  (Variables|Default|SetNPC%20PreviousMovementMode PrevMovementMode) ; Track these for Death Montage selection.\nDying immediately sets Movement Mode to None, \nso we need to know what we WERE doing.
  (Variables|Default|SetNPC%20PreviousCustomMovementMode PreviousCustomMode))

(event Custom|PlayEnterWaterSplash
  (params (Water%20Body :type Fort%20Water%20Body%20Actor%20Object%20Reference))
  ;[0 0.45 1] VFX
  ;;[1 1 1] Swimming GE Application and Remove
  (CallFunction|PlayWaterSplashAtWaterSurface Water%20Body (Variables|NPCSwimming|GetEnter%20Water)))

(event WaterBody|EventOnEnteredWaterBodyClient
  (params (WaterBody :type Fort%20Water%20Body%20Actor%20Object%20Reference) (bIsFirstBody :type Boolean))
  ;[0 0.45 1] VFX
  ;;[1 1 1] Swimming GE Application and Remove
  (if (Variables|NPCSwimming|GetNPC%20Client%20Apply%20Water%20Enter/Exit%20Burst%20GC)
    (CallFunction|PlayWaterSplashAtWaterSurface WaterBody (Variables|NPCSwimming|GetEnter%20Water))))

(event WaterBody|EventOnEnteredWaterBody
  (params (WaterBody :type Fort%20Water%20Body%20Actor%20Object%20Reference) (bIsFirstBody :type Boolean))
  (bind _npc_ge_classtoapplyforswimming (Variables|NPCSwimming|GetNPC%20GE_ClassToApplyForSwimming))
  ;[0 0.45 1] VFX
  ;;[1 1 1] Swimming GE Application and Remove
  (Default|CallNPCEnteredWaterBody WaterBody true)
  (Variables|Sound|SetIsEnteringWater true)
  (CallFunction|UpdateSoundLibFoley)
  (CallFunction|TryWaterInteractionDrown false)
  (if (Variables|NPCSwimming|GetNPC%20Apply%20Water%20Enter/Exit%20Burst%20GC)
    (CallFunction|PlayEnterWaterSplash WaterBody))
  (if (Variables|NPCSwimming|GetNPC%20Apply%20GE%20for%20Swimming)
    (if (Utilities|IsValidClass _npc_ge_classtoapplyforswimming)
      (bind _returnvalue (GameplayEffects|ApplyGameplayEffectToSelf :self (Variables|GameplayAbilities|GetAIPawnAbilitySystemComponent) :GameplayEffectClass _npc_ge_classtoapplyforswimming :Level 0.0))
      (Variables|NPCSwimming|SetNPC%20GE_AppliedForSwimming _returnvalue))))

(event WaterBody|EventOnEndDeepWaterInteraction
  (params (PreviousWaterSurfaceLocation :type Vector :default "0, 0, 0") (WaterDepth :type Float%20%28single-precision%29) (WaterBody :type Water%20Body%20Object%20Reference))
  ;[1 0.1868 0] Despawn / Death Handling
  ;;[1 1 1] Deep Water Interaction
  (if (Variables|NPCSwimming|GetNPC%20Apply%20Deep%20Water%20Enter/Exit%20Burst%20GC)
    (bind _asfort_water_body_actor (Utilities|Casting|CastToFortWaterBodyActor WaterBody))
    (CallFunction|PlayWaterSplashAtWaterSurface _asfort_water_body_actor (Variables|NPCSwimming|GetExit%20Deep%20Water))))

(event Custom|PlayExitWaterSplash
  (params (Water%20Body :type Fort%20Water%20Body%20Actor%20Object%20Reference))
  ;[0 0.45 1] VFX
  ;;[1 1 1] Swimming GE Application and Remove
  (CallFunction|PlayWaterSplashAtWaterSurface Water%20Body (Variables|NPCSwimming|GetExit%20Water)))

(event WaterBody|EventOnExitedWaterBodyClient
  (params (WaterBody :type Fort%20Water%20Body%20Actor%20Object%20Reference) (bIsLastBody :type Boolean))
  ;[0 0.45 1] VFX
  ;;[1 1 1] Swimming GE Application and Remove
  (if (Variables|NPCSwimming|GetNPC%20Client%20Apply%20Water%20Enter/Exit%20Burst%20GC)
    (CallFunction|PlayWaterSplashAtWaterSurface WaterBody (Variables|NPCSwimming|GetExit%20Water))))

(event WaterBody|EventOnExitedWaterBody
  (params (WaterBody :type Fort%20Water%20Body%20Actor%20Object%20Reference) (bIsLastBody :type Boolean))
  ;[0 0.45 1] VFX
  ;;[1 1 1] Swimming GE Application and Remove
  (Default|CallNPCEnteredWaterBody WaterBody)
  (Variables|Sound|SetIsEnteringWater false)
  (CallFunction|UpdateSoundLibFoley)
  (if (Variables|NPCSwimming|GetNPC%20Apply%20Water%20Enter/Exit%20Burst%20GC)
    (CallFunction|PlayExitWaterSplash WaterBody))
  (if (Variables|NPCSwimming|GetNPC%20Apply%20GE%20for%20Swimming)
    (if (Utilities|IsValidClass (Variables|NPCSwimming|GetNPC%20GE_ClassToApplyForSwimming))
      (GameplayEffects|RemoveActiveGameplayEffect (Variables|GameplayAbilities|GetAIPawnAbilitySystemComponent) (Variables|NPCSwimming|GetNPC%20GE_AppliedForSwimming) -1))))

(event Custom|InitializeTestInSafeZone
  (bind _aipawnabilitysystemcomponent (Variables|GameplayAbilities|GetAIPawnAbilitySystemComponent))
  ;[1 0.1868 0] Despawn / Death Handling
  ;;[1 1 1] Safe Zone Handling
  ;;;[1 1 1] Init Safe Zone Test
  (bind _asbool (CallFunction|NPCGetScalableFloatHotfixValue (Variables|NPCDespawnfromStorm|GetNPC%20ShouldDespawnFromStormHF)))
  (if _asbool
    (Networking|SwitchHasAuthority
      (:Authority
        (bind _returnvalue (Utilities|Time|SetTimerbyEvent (EventDispatchers|CreateEvent :!event Custom|TestIsInSafeZone) (Math|Random|RandomFloatinRange 2.5 3.5) true))
        (Variables|NPCDespawnfromStorm|SetNPC%20TestIsInSafeZoneTimer _returnvalue)
        (GameplayEffects|ApplyGameplayEffectToSelf :self _aipawnabilitysystemcomponent :GameplayEffectClass "/Game/Athena/AI/Abilities/NPC/Parents/GE_InSafeZone.GE_InSafeZone_C" :Level 0.0)
        (if (not (Variables|NPCDespawnfromStorm|GetIsGEStormDamageInit))
          (bind _returnvalue_2 (GameplayEffects|ApplyGameplayEffectToSelf :self _aipawnabilitysystemcomponent :GameplayEffectClass "/Game/Athena/SafeZone/GE_OutsideSafeZoneDamage.GE_OutsideSafeZoneDamage_C" :Level 1.0))
          (bind _output_get (Variables|NPCDespawnfromStorm|SetGeOutsideSafeZoneDamageHandle _returnvalue_2))
          (Ability|GameplayEffect|AssignActiveGameplayEffectTagSetbyCallerMagnitude _output_get "(TagName=\"SetByCaller.StormCampingDamage\")" 1.0)
          (Variables|NPCDespawnfromStorm|SetIsGEStormDamageInit true))))))

(event Custom|TryWaterInteractionDrown
  (params (CalledFromDeepWaterInteraction :type Boolean))
  ;[1 0.1868 0] Despawn / Death Handling
  ;;[1 1 1] Triggered instant drown
  ;; Of course, this will only happen if the pawn has the correct ability type to trigger
  (if (not CalledFromDeepWaterInteraction)
    (bind _asbool (CallFunction|NPCGetScalableFloatHotfixValue (Variables|NPCSwimming|GetNPC%20AllowDrownInShallowWaterHF)))
    (if _asbool
      _))
  (bind _aipawnabilitysystemcomponent (Variables|GameplayAbilities|GetAIPawnAbilitySystemComponent)
    (:then
      (Abilities|TryActivateAbilitiesbyTag _aipawnabilitysystemcomponent (GameplayTags|MakeLiteralGameplayTagContainer "(GameplayTags=((TagName=\"NPC.Ability.NonAttack.Death.Drowned\")))") true))
    (:else)))

(event Custom|LuredOnServer
  (params (AlertLevel :type Enum_NPC_AlertLevel%20Enum :default "NewEnumerator4"))
  (bind _self self)
  (bind _returnvalue (AI|GetBlackboard _self))
  (bind _goalactorbbkey (Variables|NPCLure|GetGoalActorBBKey))
  (bind _returnvalue_2 (AI|Components|Blackboard|GetValueasObject _returnvalue _goalactorbbkey))
  ;[0 0.45 1] VFX
  ;;[1 1 1] Handle  Pawn Being Lured to an item (basic implementation)
  (bind _asbool (CallFunction|NPCGetScalableFloatHotfixValue (Variables|NPCLure|GetNPC%20PlayEffectWhenLuredHF)))
  (if _asbool
    (switch Utilities|FlowControl|Switch|SwitchonEnum_NPC_AlertLevel AlertLevel
      (:NewEnumerator1
        (Utilities|IsValid _returnvalue_2
          (:Is%20Valid
            (if (Utilities|NotEqual(Object) (Variables|NPCLure|GetNPC%20PreviousLureGoal) _returnvalue_2)
              (Variables|NPCLure|SetNPC%20PreviousLureGoal _returnvalue_2)
              (Utilities|Casting|CastToFortPickup _returnvalue_2
                (:then
                  _)
                (:CastFailed
                  (Utilities|Casting|CastToFortProjectileBase _returnvalue_2)))
              (CallFunction|LuredInterest)))
          (:Is%20Not%20Valid
            (Variables|NPCLure|SetNPC%20PreviousLureGoal 0)))))))

(event Custom|LuredInterest
  (bind _gameplaycuelureinterest (Variables|NPCLure|GetGameplayCueLureInterest))
  ;[0 0.45 1] VFX
  ;;[1 1 1] Handle  Pawn Being Lured to an item (basic implementation)
  (Networking|CanExecuteCosmeticEvents
    (:True
      (if (GameplayTags|IsGameplayTagValid _gameplaycuelureinterest)
        (GameplayCue|ExecuteGameplayCue%28Burst%29Local self _gameplaycuelureinterest (Ability|GameplayCue|MakeGameplayCueParameters 0.0 (Variables|Internal|GetNPC%20Skeletal%20Mesh%20Index)))))))

(event Custom|TestIsInSafeZone
  (bind _aipawnabilitysystemcomponent (Variables|GameplayAbilities|GetAIPawnAbilitySystemComponent))
  ;[1 0.1868 0] Despawn / Death Handling
  ;;[1 1 1] Safe Zone Handling
  ;;;[1 1 1] Safe Zone Test Loop
  (bind (_testsucceeded _isinsafezone) (CallFunction|NPCCheckIsinSafeZone))
  (if _testsucceeded
    (if _isinsafezone
      (GameplayEffects|ApplyGameplayEffectToSelf :self _aipawnabilitysystemcomponent :GameplayEffectClass "/Game/Athena/AI/Abilities/NPC/Parents/GE_InSafeZone.GE_InSafeZone_C" :Level 0.0)
      (else
        (GameplayEffects|RemoveActiveGameplayEffectbySourceEffect _aipawnabilitysystemcomponent "/Game/Athena/AI/Abilities/NPC/Parents/GE_InSafeZone.GE_InSafeZone_C" _aipawnabilitysystemcomponent -1)
        (bind _safezonephase (CallFunction|NPCGetSafeZonePhase))
        (GameplayEffects|SetActiveGameplayEffectLevel _aipawnabilitysystemcomponent (Variables|NPCDespawnfromStorm|GetGeOutsideSafeZoneDamageHandle) _safezonephase)))))

(event EventOnDeathServer
  (params (Damage :type Float%20%28single-precision%29) (DamageTags :type Gameplay%20Tag%20Container%20Structure%20%28by%20ref%29) (Momentum :type Vector :default "0, 0, 0") (HitInfo :type Hit%20Result%20Structure%20%28by%20ref%29) (InstigatedBy :type Controller%20Object%20Reference) (DamageCauser :type Actor%20Object%20Reference) (EffectContext :type Gameplay%20Effect%20Context%20Handle%20Structure))
  ;[1 0.1868 0] Despawn / Death Handling
  ;;[1 1 1] On Death Server (clear loot drops)
  (bind _result (CallFunction|ShouldRunInstantDeathCue DamageTags InstigatedBy DamageCauser))
  (Variables|Default|SetRunningInstantDeathCue _result)
  (CallFunction|NPCClearOutTargetSlots)
  (Variables|Default|SetNPC%20Death%20Time (Utilities|Time|GetGameTimeinSeconds))
  (bind (_asbool _asintrounded _asfloat) (CallFunction|NPCGetScalableFloatHotfixValue (Variables|NPCInventory|GetNPC%20DeathDropLootDelay)))
  (if (> _asfloat 0.0)
    (Utilities|FlowControl|Delay _asfloat))
  (CallFunction|NPCTrytoDropLootonDeath DamageTags))

(event Custom|TryToDespawnDueToStorm
  (params (Reset%20Timer :type Boolean))
  ;[1 0.1868 0] Despawn / Death Handling
  ;;[1 1 1] Safe Zone Handling
  ;;;[1 1 1] Try Despawn From Storm (Outside Safezone)
  (if (or (and Reset%20Timer (Utilities|Time|IsValidTimerHandle (Variables|NPCDespawnfromStorm|GetNPC%20DespawnFromStormTimer))) (not (Utilities|Time|IsValidTimerHandle (Variables|NPCDespawnfromStorm|GetNPC%20DespawnFromStormTimer))))
    (Variables|NPCDespawnfromStorm|SetNPC%20isTryingToDespawn true)
    (bind (_asbool _asintrounded _asfloat) (CallFunction|NPCGetScalableFloatHotfixValue (Variables|NPCDespawnfromStorm|GetNPC%20DespawnTimerHF)))
    (bind _returnvalue (Utilities|Time|SetTimerbyEvent (EventDispatchers|CreateEvent :!event Custom|DespawnDueToStorm) _asfloat))
    (Variables|NPCDespawnfromStorm|SetNPC%20DespawnFromStormTimer _returnvalue)))

(event EventOnDamagePlayEffects
  (params (Damage :type Float%20%28single-precision%29) (DamageTags :type Gameplay%20Tag%20Container%20Structure%20%28by%20ref%29) (Momentum :type Vector :default "0, 0, 0") (HitInfo :type Hit%20Result%20Structure%20%28by%20ref%29) (InstigatedBy :type Fort%20Pawn%20Object%20Reference) (DamageCauser :type Actor%20Object%20Reference) (EffectContext :type Gameplay%20Effect%20Context%20Handle%20Structure))
  ;[0 0.45 1] VFX
  ;;[1 1 1] Additive Hit Reactions and Headshot VFX/Sound
  (Utilities|FlowControl|Sequence
    (:then_0
      (Variables|NPCAdditiveHitReactions|GetAdditive%20Hit%20React%20Montage
        (:then
          (if (Math|Boolean|ANDBoolean (not (Variables|Animation|GetbIsStaggered)) (not (Variables|Animation|GetbIsKnockedback)) (not (GameplayTags|HasAnyMatchingGameplayTags (Variables|HitReaction|GetGameplayTagsToBlockHitReact))) (not (Variables|Health|GetbIsDying)))
            (if (> (Variables|NPCAdditiveHitReactions|GetAdditiveHitReactRetriggerDelay) 0)
              (Utilities|FlowControl|DoOnce false
                (:Completed
                  (Utilities|Time|SetTimerbyFunctionName 0 "AdditiveHitReactDelay" (Variables|NPCAdditiveHitReactions|GetAdditiveHitReactRetriggerDelay)))))
            (bind _additive_hit_react_montage (Variables|NPCAdditiveHitReactions|GetAdditive%20Hit%20React%20Montage)
              (:then
                (CallFunction|PlayAdditiveHitReacts Momentum _additive_hit_react_montage) ; Now we play the overridable function, override it by playing the Native function below.\n\nDouble click for the overridden logic.
              )
              (:else))))
        (:else)))
    (:then_1
      (switch Utilities|FlowControl|Switch|SwitchonEFortDamageZone (DamageZones|GetDamageZone HitInfo)
        (:Critical
          (if (Variables|NPCHeadshotVFX|GetUseHeadShotFX)
            (bind _asfort_weapon (Utilities|Casting|CastToFortWeapon DamageCauser)
              (:then
                (if (Tags|WeaponDataHasAnyMatchingGameplayTags _asfort_weapon (Variables|NPCHeadshotVFX|GetTC_WeaponTypesAllowedForHeadshotFX))
                  (Variables|NPCHeadshotVFX|SetAllowHeadshot true :!label _label_2)
                  (else
                    (Variables|NPCHeadshotVFX|SetAllowHeadshot false :!label _label)))
                (if (Variables|NPCHeadshotVFX|GetAllowHeadshot)
                  (Utilities|FlowControl|Sequence
                    (:then_0
                      (bind _headshotfxemittertemplate (Variables|NPCHeadshotVFX|GetHeadshotFXEmitterTemplate)
                        (:then
                          (Effects|SpawnEmitteratLocation _headshotfxemittertemplate (Transformation|GetSocketLocation (Variables|Character|GetMesh) (Variables|NPCHeadshotVFX|GetHeadshot%20FX%20Socket)) (Transformation|GetWorldRotation (Variables|Character|GetMesh))))
                        (:else)))
                    (:then_1
                      (bind _headshot_sound (Variables|Sound|GetHeadShot_Sound)
                        (:then
                          (Audio|SpawnSoundAttached _headshot_sound (Variables|Character|GetMesh)))
                        (:else))))))
              (:CastFailed
                (Utilities|Casting|CastToFortProjectileBase DamageCauser
                  (:then
                    _)
                  (:CastFailed
                    (goto _label)))))
            (goto _label_2))
          (goto _label))))))

(event EventOnDeathPlayEffects
  (params (Damage :type Float%20%28single-precision%29) (DamageTags :type Gameplay%20Tag%20Container%20Structure%20%28by%20ref%29) (Momentum :type Vector :default "0, 0, 0") (HitInfo :type Hit%20Result%20Structure%20%28by%20ref%29) (InstigatedBy :type Fort%20Pawn%20Object%20Reference) (DamageCauser :type Actor%20Object%20Reference) (EffectContext :type Gameplay%20Effect%20Context%20Handle%20Structure))
  ;[1 0.1868 0] Despawn / Death Handling
  ;;[1 1 1] On Death Effects
  (bind _result (CallFunction|ShouldRunInstantDeathCue DamageTags InstigatedBy DamageCauser))
  (if (Math|Boolean|ANDBoolean (GameplayTags|HasTag DamageTags (Variables|Default|GetGCInstantDeathTag)) (not (Utilities|IsValid InstigatedBy :!class /Script/Engine.KismetSystemLibrary)) (not (Utilities|IsValid DamageCauser :!class /Script/Engine.KismetSystemLibrary)))
    (CallFunction|NPCDisableHealthBars))
  (bind _success (UEFNCustomEffects|PlayCustomEffect "(TagName=\"NPC.CustomEffect.DeathEffect\")" self))
  (if (not _success)
    (Audio|SpawnSoundAttached (Variables|Sound|GetDeath%20Normal%20Sound) (Variables|Character|GetCapsuleComponent)) ; Normal Death Sound
    (if (GameplayTags|NotEqual(GameplayTag) :A (Variables|NPCDeath|GetGameplayCue_Death))
      (bind _output_get (Variables|Default|SetRunningInstantDeathCue _result))
      (CallFunction|NPCMulticastGameplayCueDeath _output_get))))

(event Custom|AdditiveHitReactDelay
  ;[0 0.45 1] VFX
  ;;[1 1 1] Additive Hit Reactions and Headshot VFX/Sound
  (Utilities|FlowControl|DoOnce false
    (:Completed
      (Utilities|Time|SetTimerbyFunctionName 0 "AdditiveHitReactDelay" (Variables|NPCAdditiveHitReactions|GetAdditiveHitReactRetriggerDelay))
      (bind _additive_hit_react_montage (Variables|NPCAdditiveHitReactions|GetAdditive%20Hit%20React%20Montage)
        (:then
          (CallFunction|PlayAdditiveHitReacts (stale EventOnDamagePlayEffects Momentum) _additive_hit_react_montage) ; Now we play the overridable function, override it by playing the Native function below.\n\nDouble click for the overridden logic.
        )
        (:else)))))

(event Custom|DespawnDueToStorm
  ;[1 0.1868 0] Despawn / Death Handling
  ;;[1 1 1] Safe Zone Handling
  ;;;[1 1 1] Despawn
  (if (Variables|NPCDespawnfromStorm|GetNPC%20TryDespawnAbility)
    (CallFunction|TriggerDespawnAbility)
    (else
      (FortPawn|ForceKill (Variables|NPCDespawnFromStorm|GetNPC%20DeathByStormTag)))))

(event Custom|TriggerDespawnAbility
  ;[1 0.1868 0] Despawn / Death Handling
  ;;[1 1 1] Safe Zone Handling
  ;;;[1 1 1] Despawn
  (Ability|SendGameplayEventtoActor self (Variables|NPCDespawnfromStorm|GetNPC%20DespawnGameplayTag) (Utilities|Struct|MakeGameplayEventData (Variables|NPCDespawnfromStorm|GetNPC%20DespawnGameplayTag))))

(event Custom|NPC%20MulticastGameplayCueDeath
  (params (ShouldRunInstantDeathCue :type Boolean))
  (bind _returnvalue (Ability|GameplayCue|MakeGameplayCueParameters 0.0))
  ;[1 0.1868 0] Despawn / Death Handling
  ;;[1 1 1] On Death Effects
  (Networking|CanExecuteCosmeticEvents
    (:True
      (bind _output_get (Variables|Default|SetRunningInstantDeathCue ShouldRunInstantDeathCue))
      (if _output_get
        (if (CallFunction|DespawnsonInstantDeath)
          (GameplayCue|ExecuteGameplayCue%28Burst%29Local self (Variables|NPCDeath|GetGameplayCueDespawnOnInstantDeath) _returnvalue)
          (else
            (GameplayCue|ExecuteGameplayCue%28Burst%29Local self (GameplayTags|MakeLiteralGameplayTag (Variables|NPCDeath|GetGameplayCue_SystemInstantDeath)) _returnvalue)))
        (else
          (GameplayCue|ExecuteGameplayCue%28Burst%29Local self (GameplayTags|MakeLiteralGameplayTag (Variables|NPCDeath|GetGameplayCue_Death)) _returnvalue))))))

(event Custom|StopTestingIsInSafeZone
  (bind _npc_despawnfromstormtimer (Variables|NPCDespawnfromStorm|GetNPC%20DespawnFromStormTimer))
  (bind _npc_testisinsafezonetimer (Variables|NPCDespawnfromStorm|GetNPC%20TestIsInSafeZoneTimer))
  ;[1 0.1868 0] Despawn / Death Handling
  ;;[1 1 1] Safe Zone Handling
  ;;;[1 1 1] Clear out timers and flags related to despawning from storm (call InitializeTestInSafeZone to start again)
  (Variables|NPCDespawnfromStorm|SetNPC%20isTryingToDespawn false)
  (Utilities|FlowControl|Sequence
    (:then_0
      (if (Utilities|Time|IsValidTimerHandle _npc_testisinsafezonetimer)
        (Utilities|Time|ClearandInvalidateTimerbyHandle _npc_testisinsafezonetimer)))
    (:then_1
      (if (Utilities|Time|IsValidTimerHandle _npc_despawnfromstormtimer)
        (Utilities|Time|ClearandInvalidateTimerbyHandle _npc_despawnfromstormtimer)))))

(fn CanUpdateSoundLib
  (params (TargetArray :type Array%20of%20Sound%20Library%20Object%20References))
  (if (Utilities|Array|IsNotEmpty TargetArray)
    (return (and (!= (Variables|Sound|GetCurrentSoundLib) (Utilities|Array|Get(acopy) TargetArray)) (!= (Utilities|Array|Get(acopy) TargetArray) 0)))
    (else
      (return false))))

(fn CheckTag
  (params (Tag%20Container :type Gameplay%20Tag%20Container%20Structure))
  (if (Utilities|GameplayTagContainerNotEmpty Tag%20Container)
    (return false)
    (else
      (return true))))

(fn CheckSoundLibraryAdded
  (params (Libraries :type Array%20of%20Sound%20Library%20Object%20References))
  (if (Utilities|Array|IsValidIndex Libraries)
    (SoundLibrary|AddLibraries (WorldSubsystems|GetSoundLibrarySubsystem) self Libraries :!class /Script/SoundLibrary.SoundLibrarySubsystem)
    (Variables|Sound|SetCurrentSoundLib (Utilities|Array|Get(acopy) Libraries))))

(fn CheckSoundLibraryRemoved
  (params (Libraries :type Array%20of%20Sound%20Library%20Object%20References))
  (if (Utilities|Array|IsNotEmpty Libraries)
    (SoundLibrary|RemoveLibraries (WorldSubsystems|GetSoundLibrarySubsystem) self Libraries :!class /Script/SoundLibrary.SoundLibrarySubsystem)))

(fn DespawnsOnInstantDeath
  (return (GameplayTags|IsGameplayTagValid (Variables|NPCDeath|GetGameplayCueDespawnOnInstantDeath))))

(fn GetSoundFoleyLibrary
  (bind _returnvalue (WorldSubsystems|GetSoundLibrarySubsystem))
  (bind _returnvalue_2 (WorldSubsystems|GetSoundLibrarySubsystem))
  (bind _self self)
  (bind _self_2 self)
  (bind _soundlibs_foley (Variables|SoundLibraries|GetSoundLibs_Foley))
  (bind _soundlibs_foley_water (Variables|SoundLibraries|GetSoundLibs_Foley_Water))
  (if (Utilities|Array|IsValidIndex (Variables|SoundLibraries|GetSoundLibs_Foley_Water))
    (Utilities|IsValid (WorldSubsystems|GetSoundLibrarySubsystem)
      (:Is%20Valid
        (if (Variables|Sound|GetIsEnteringWater)
          (if (!= (Variables|Sound|GetCurrentFoleyLib) (Utilities|Array|Get(acopy) (Variables|SoundLibraries|GetSoundLibs_Foley_Water)))
            (SoundLibrary|RemoveLibraries _returnvalue _self (Variables|SoundLibraries|GetSoundLibs_Foley) :!class /Script/SoundLibrary.SoundLibrarySubsystem)
            (SoundLibrary|AddLibraries _returnvalue _self _soundlibs_foley_water :!class /Script/SoundLibrary.SoundLibrarySubsystem)
            (Variables|Sound|SetCurrentFoleyLib (Utilities|Array|Get(acopy) _soundlibs_foley_water)))
          (elif (!= (Variables|Sound|GetCurrentFoleyLib) (Utilities|Array|Get(acopy) (Variables|SoundLibraries|GetSoundLibs_Foley)))
            (SoundLibrary|RemoveLibraries _returnvalue_2 _self_2 (Variables|SoundLibraries|GetSoundLibs_Foley_Water) :!class /Script/SoundLibrary.SoundLibrarySubsystem)
            (SoundLibrary|AddLibraries _returnvalue_2 _self_2 _soundlibs_foley :!class /Script/SoundLibrary.SoundLibrarySubsystem)
            (Variables|Sound|SetCurrentFoleyLib (Utilities|Array|Get(acopy) _soundlibs_foley))))))))

(fn IsConverted
  (bind _self self)
  (bind _returnvalue (Actor|GetComponentbyClass _self "/Script/FortniteGame.FortPawnComponent_Convert"))
  (Utilities|IsValid _returnvalue
    (:Is%20Valid
      (return (Utilities|IsValid (Convert|GetConverterPawn _returnvalue) :!class /Script/Engine.KismetSystemLibrary)))
    (:Is%20Not%20Valid
      (return false))))

(fn NPCAlertLevelUpdatedClient
  (CallFunction|NPCAlertLevelChangedClient (Variables|NPCAlertLevel|GetNPC_AlertLevel)))

(fn NPCApplyGameplayEffectWithMMRScaling
  (params (Actor :type Actor%20Object%20Reference) (GE%20Class :type Gameplay%20Effect%20Class%20Reference))
  (Utilities|IsValid Actor
    (:Is%20Valid
      _)
    (:Is%20Not%20Valid
      _))
  (bind _data (Utilities|FlowControl|MergeExecutionwithData Actor self)
    (:Then
      (LocalVariables|SetApplyGeToThisActor _data))))

(fn NPCApplyRecentlyDamagedTagGE
  (params (RecentlyDamagedOverride :type Gameplay%20Effect%20Class%20Reference))
  (bind _ge_npc_recentlydamaged (Variables|NPCRecentlyDamaged|GetGE_NPC_RecentlyDamaged))
  (Networking|SwitchHasAuthority
    (:Authority
      (bind _aipawnabilitysystemcomponent (Variables|GameplayAbilities|GetAIPawnAbilitySystemComponent)
        (:then
          (if (Utilities|IsValidClass RecentlyDamagedOverride)
            (GameplayEffects|ApplyGameplayEffectToTarget :self _aipawnabilitysystemcomponent :GameplayEffectClass RecentlyDamagedOverride :Target _aipawnabilitysystemcomponent :Level 0.0) ; Apply an override GE in this case?
            (elif (Utilities|IsValidClass _ge_npc_recentlydamaged)
              (GameplayEffects|ApplyGameplayEffectToTarget :self _aipawnabilitysystemcomponent :GameplayEffectClass _ge_npc_recentlydamaged :Target _aipawnabilitysystemcomponent :Level 0.0) ; Apply our standard Recently Damaged GE?
            )))
        (:else)))))

(fn NPCAttemptToSetLeader
  (params (PossibleLeader :type Actor%20Object%20Reference) (UnconvertReason :type EUnconvertReason%20Enum :default "StealByOtherPlayer"))
  ;[1 1 1] Between these two nodes we can have our go/nogo logic.
  (bind (_success _leaderset) (CallFunction|NPCSetLeader PossibleLeader))
  (CallFunction|NPCLeaderSet _success _leaderset UnconvertReason))

(fn NPCCanBeInteractedWith
  (return (Variables|NPCInteraction|GetNPC%20PawnCanInteractInternal)))

(fn NPCCanHideOrUnHidePawnOnSpawn
  (return (and (not (Math|Boolean|ORBoolean (Variables|Health|GetbIsDying) (FortPawn|IsDead) (Variables|Health|GetbIsHiddenForDeath))) (or (Variables|NPCSpawning|GetNPC%20SpawnInvisible) (Variables|NPCSpawning|GetNPC%20ScaleMeshInOnSpawn)))))

(fn NPCCheckForDespawnIfStaleToPlayers
  (params (ClosestFortPlayerPawn :type Fort%20Player%20Pawn%20Object%20Reference))
  (Utilities|IsValid ClosestFortPlayerPawn
    (:Is%20Valid
      (bind (_asbool _asintrounded _asfloat) (CallFunction|NPCGetScalableFloatHotfixValue (Variables|NPCDespawnIfStaleToPlayers|GetNPC%20StaleDespawnTestDistance%20HF)))
      (bind (_asbool_2 _asintrounded_2 _asfloat_2) (CallFunction|NPCGetScalableFloatHotfixValue (Variables|NPCDespawnIfStaleToPlayers|GetNPC%20StaleDespawnTimeThreshold%20HF)))
      (if (and (> _asfloat_2 0.0) (> _asfloat 0.0))
        (if (> (Class|AbilityFunctions|GetDistanceBetweenActors ClosestFortPlayerPawn self) _asfloat)
          (if (> (Variables|NPCDespawnIfStaleToPlayers|GetNPC%20LastNonStaleCheckTime) 0.0)
            (if (> (- (Utilities|Time|GetGameTimeinSeconds) (Variables|NPCDespawnIfStaleToPlayers|GetNPC%20LastNonStaleCheckTime)) _asfloat_2)
              (CallFunction|TriggerDespawnAbility))
            (else
              _)))
        (Variables|NPCDespawnIfStaleToPlayers|SetNPC%20LastNonStaleCheckTime (Utilities|Time|GetGameTimeinSeconds)) ; This is the time we either first realized we were Stale, \nor the last time we were NOT Stale
      ))))

(fn NPCCheckIsInSafeZone
  (bind _returnvalue (SafeZone|GetSafeZoneIndicator))
  (if (Utilities|ClassIsChildOf (Utilities|GetClass (Game|GetGameState)) (Utilities|ResolveSoftReference (Variables|Default|GetAthena_GameState_SoftClass)))
    (Utilities|IsValid _returnvalue
      (:Is%20Valid
        (return true (< (Math|Vector|Distance2DSquared(Vector) (SafeZone|GetSafeZoneCenter _returnvalue) (Transformation|GetActorLocation)) (Math|Float|Square (SafeZone|GetSafeZoneRadius _returnvalue)))))
      (:Is%20Not%20Valid
        (return false false)))
    (else
      (return false false))))

(fn NPCClearOutTargetSlots
  (bind _self self)
  (bind _returnvalue (GameplayTags|GetOwnedGameplayTags _self :!class /Script/GameplayTags.GameplayTagAssetInterface))
  (bind _self_2 self)
  ;[1 1 1] Make sure we clear out any Target Slots we are occupying when we die or disappear.
  (if (GameplayTags|HasAnyTags _returnvalue (GameplayTags|MakeLiteralGameplayTagContainer "(GameplayTags=((TagName=\"NPC.Behavior.TargetSlots.Melee.SlotOwner\"),(TagName=\"NPC.Behavior.TargetSlots.Ranged.SlotOwner\")))"))
    (Utilities|FlowControl|Sequence
      (:then_0
        (if (GameplayTags|HasTag _returnvalue "(TagName=\"NPC.Behavior.TargetSlots.Melee.SlotOwner\")" true)
          (Ability|SendGameplayEventtoActor _self_2 "(TagName=\"NPC.Behavior.TargetSlots.Melee.Remove\")")))
      (:then_1
        (if (GameplayTags|HasTag _returnvalue "(TagName=\"NPC.Behavior.TargetSlots.Ranged.SlotOwner\")" true)
          (Ability|SendGameplayEventtoActor _self_2 "(TagName=\"NPC.Behavior.TargetSlots.Ranged.Remove\")"))))))

(fn NPCDamageAttemptDisableIntroPhaseBehavior
  (params (DamageTags :type Gameplay%20Tag%20Container%20Structure))
  (bind _tq_npc_damagetagsthatcancancelintrophase (Variables|NPCIntroPhase|GetTQ_NPC%20DamageTagsThatCanCancelIntroPhase))
  (Networking|SwitchHasAuthority
    (:Authority
      (if (Variables|NPCIntroPhase|GetNPC%20DamageCanCancelIntroPhase)
        (if (not (GameplayTags|IsTagQueryEmpty _tq_npc_damagetagsthatcancancelintrophase))
          (if (GameplayTags|DoesContainerMatchTagQuery DamageTags _tq_npc_damagetagsthatcancancelintrophase)
            _))
        (CallFunction|NPCDisableIntroPhaseBehavior)))))

(fn NPCDamageRemovesIdlePhaseTagGE)

(fn NPCDisableIntroPhaseBehavior
  (bind _ge_npc_introphasebehaviorcooldown (Variables|NPCIntroPhase|GetGE_NPC_IntroPhaseBehaviorCooldown))
  (bind _aipawnabilitysystemcomponent (Variables|GameplayAbilities|GetAIPawnAbilitySystemComponent)
    (:then
      (GameplayEffects|RemoveActiveEffectswithAppliedTags _aipawnabilitysystemcomponent (Variables|NPCIntroPhase|GetTC_IntroBehaviorActive))
      (if (Utilities|IsValidClass _ge_npc_introphasebehaviorcooldown)
        (GameplayEffects|ApplyGameplayEffectToTarget :self _aipawnabilitysystemcomponent :GameplayEffectClass _ge_npc_introphasebehaviorcooldown :Target _aipawnabilitysystemcomponent :Level 0.0) ; Make sure we apply the cooldown in case we DON'T have the Intro Phase Active GE yet.\n\nShould keep goals add by damage from iniating the Intro Phase Behavior.
      ))
    (:else)))

(fn NPCGetAlertLevelInfo
  (bind (_value _returnvalue) (Utilities|Map|Find (Variables|NPCAlertLevel|GetNPC%20AlertLevelToTagMap) (Variables|NPCAlertLevel|GetNPC_AlertLevel)))
  (return (Variables|NPCAlertLevel|GetNPC_AlertLevel) _value _returnvalue))

(fn NPCGetInventoryInformation
  (if (Variables|Inventory|GetbHasInventory)
    (bind _items (InventoryManagement|GetItemsfromInventoryOwner self))
    (bind (_asbool _asintrounded) (CallFunction|NPCGetScalableFloatHotfixValue (Variables|NPCInventory|GetNPC%20NumberOfInventoryItemsToDeliverHF))) ; Get how many items we need before we would return them.
    (bind (_asbool_2 _asintrounded_2) (CallFunction|NPCGetScalableFloatHotfixValue (Variables|NPCInventory|GetNPC%20NumberOfInventoryItemsToConsumeBeforeFullHF)))
    (CallFunction|NPCPawnDebugMessage (Utilities|String|BuildString(Boolean) (Utilities|String|BuildString(Boolean) (Utilities|String|BuildString(Boolean) (Utilities|String|BuildString(Boolean) "Get Inventory Info | " "Care? = " (> _asintrounded 0) " | ") "Enough to Deliver? = " (>= (Utilities|Array|Length _items) _asintrounded) " | ") "Has Inv? = " (> (Utilities|Array|Length _items) 0) " | ") "DELIVER RESULT = " (Math|Boolean|ANDBoolean (> _asintrounded 0) (>= (Utilities|Array|Length _items) _asintrounded) (> (Utilities|Array|Length _items) 0)) " ====="))
    (return (> (Utilities|Array|Length _items) 0) (Utilities|Array|Length _items) (Utilities|Array|LastIndex _items) (Math|Boolean|ANDBoolean (> _asintrounded 0) (>= (Utilities|Array|Length _items) _asintrounded) (> (Utilities|Array|Length _items) 0)) (>= (Utilities|Array|Length _items) _asintrounded_2))))

(fn NPCGetPreviousAlertLevelInfo
  (bind (_value _returnvalue) (Utilities|Map|Find (Variables|NPCAlertLevel|GetNPC%20AlertLevelToTagMap) (Variables|NPCAlertLevel|GetNPC_AlertLevelPrevious)))
  (return (Variables|NPCAlertLevel|GetNPC_AlertLevelPrevious) _value _returnvalue))

(fn NPCGetSafeZonePhase
  (bind _returnvalue (SafeZone|GetSafeZoneIndicator))
  (Utilities|IsValid _returnvalue
    (:Is%20Valid
      (return (SafeZone|GetCurrentPhaseIndex _returnvalue)))
    (:Is%20Not%20Valid
      (return 0))))

(fn NPCGetScalableFloatHotfixValue
  (params (ScalableFloatHotfix :type Scalable%20Float%20Structure%20%28by%20ref%29) (AtLevel :type Float%20%28double-precision%29))
  (if (CallFunction|NPCScalableFloatIsValid ScalableFloatHotfix)
    (return (> (Math|Float|Round (ScalableFloat|GetValueatLevel ScalableFloatHotfix AtLevel)) 0) (Math|Float|Round (ScalableFloat|GetValueatLevel ScalableFloatHotfix AtLevel)) (ScalableFloat|GetValueatLevel ScalableFloatHotfix AtLevel))
    (else
      (return (> (Math|Float|Round (Utilities|Struct|BreakScalableFloat ScalableFloatHotfix)) 0) (Math|Float|Round (Utilities|Struct|BreakScalableFloat ScalableFloatHotfix)) (Utilities|Struct|BreakScalableFloat ScalableFloatHotfix)))))

(fn NPCGhostCollisionSettings
  (params (NoCollisionForRanged :type Boolean) (NoCollisionForMelee :type Boolean) (NoCollisionForProjectiles :type Boolean) (OnlyOverlapCollisionForPawns :type Boolean))
  ;[1 1 1] This is all WIP
  ;
  ; We probably want to cache-off default settings when we reset collision
  (Utilities|FlowControl|Sequence
    (:then_0
      (if NoCollisionForRanged
        (bind _weaponcapsulecollision (Variables|Default|GetWeaponCapsuleCollision)
          (:then
            (Collision|SetCollisionEnabled _weaponcapsulecollision "NoCollision" :!pinned true) ; Make sure Capsule and Mesh don't collide\n\nTODO: May need to tweak on per-channel basis to retain marking/interactions, etc.
            (bind _mesh (Variables|Character|GetMesh)
              (:then
                (Collision|SetCollisionEnabled _mesh "NoCollision"))
              (:else)))
          (:else))
        (else
          (bind _weaponcapsulecollision_2 (Variables|Default|GetWeaponCapsuleCollision)
            (:then
              (Collision|SetCollisionEnabled _weaponcapsulecollision_2 "QueryOnly")
              (bind _mesh_2 (Variables|Character|GetMesh)
                (:then
                  (Collision|SetCollisionEnabled _mesh_2 "QueryOnly"))
                (:else)))
            (:else)))))
    (:then_1
      (if NoCollisionForMelee
        (bind _capsulecomponent (Variables|Character|GetCapsuleComponent)
          (:then
            (Collision|SetCollisionResponsetoChannel _capsulecomponent "ECC_GameTraceChannel10" "ECR_Ignore" :!pinned true) ; Stops melee/pickaxe collision.
            (Collision|SetCollisionResponsetoChannel _capsulecomponent "ECC_WorldDynamic" "ECR_Ignore" :!pinned true) ; Stops things like StW Dragon Slash from hitting.
          )
          (:else))
        (else
          (bind _capsulecomponent_2 (Variables|Character|GetCapsuleComponent)
            (:then
              (Collision|SetCollisionResponsetoChannel _capsulecomponent_2 "ECC_GameTraceChannel10" "ECR_Block")
              (Collision|SetCollisionResponsetoChannel _capsulecomponent_2 "ECC_WorldDynamic" "ECR_Block"))
            (:else)))))
    (:then_2
      (if NoCollisionForProjectiles
        (bind _capsulecomponent_3 (Variables|Character|GetCapsuleComponent)
          (:then
            (Collision|SetCollisionResponsetoChannel _capsulecomponent_3 "ECC_GameTraceChannel8" "ECR_Ignore" :!pinned true) ; Projectiles (and sniper rifles) do not hit.
            (Collision|SetCollisionResponsetoChannel _capsulecomponent_3 "ECC_GameTraceChannel12" "ECR_Ignore"))
          (:else))
        (else
          (bind _capsulecomponent_4 (Variables|Character|GetCapsuleComponent)
            (:then
              (Collision|SetCollisionResponsetoChannel _capsulecomponent_4 "ECC_GameTraceChannel8" "ECR_Block")
              (Collision|SetCollisionResponsetoChannel _capsulecomponent_4 "ECC_GameTraceChannel12" "ECR_Block"))
            (:else)))))
    (:then_3
      (if OnlyOverlapCollisionForPawns
        (bind _capsulecomponent_5 (Variables|Character|GetCapsuleComponent)
          (:then
            (Collision|SetCollisionResponsetoChannel _capsulecomponent_5 "ECC_Pawn" "ECR_Overlap" :!pinned true) ; Still overlap pawns so we can know we've reached something important by touch. \nWe may or may not react to this.
            (Collision|SetCollisionResponsetoChannel _capsulecomponent_5 "ECC_GameTraceChannel4" "ECR_Overlap"))
          (:else))
        (else
          (bind _capsulecomponent_6 (Variables|Character|GetCapsuleComponent)
            (:then
              (Collision|SetCollisionResponsetoChannel _capsulecomponent_6 "ECC_Pawn" "ECR_Block")
              (Collision|SetCollisionResponsetoChannel _capsulecomponent_6 "ECC_GameTraceChannel4" "ECR_Overlap"))
            (:else)))))))

(fn NPCHandleAlertLevelVoice
  (bind _npc_voicecomponent (Variables|Default|GetNPC_VoiceComponent))
  (bind _voaudiocomponent (Variables|Default|GetVOAudioComponent))
  (switch Utilities|FlowControl|Switch|SwitchonEnum_NPC_AlertLevel (Variables|NPCAlertLevel|GetNPC_AlertLevel)
    (:NewEnumerator1
      (Class|NPCVoiceComponent|PlayVoiceLine _npc_voicecomponent (Utilities|Select :Index (Variables|NPCAlertLevel|GetNPC_AlertLevelPrevious) :NewEnumerator2 (Variables|VoiceComponent|GetVoiceTag_ReturnToIdle) :NewEnumerator3 (Variables|VoiceComponent|GetVoiceTag_LostTarget)) _voaudiocomponent))
    (:NewEnumerator2
      (Class|NPCVoiceComponent|PlayVoiceLine _npc_voicecomponent (Variables|VoiceComponent|GetVoiceTag_Suspicious) _voaudiocomponent))
    (:NewEnumerator3
      (Class|NPCVoiceComponent|PlayVoiceLine _npc_voicecomponent (Variables|VoiceComponent|GetVoiceTag_FullAlert) _voaudiocomponent))))

(fn NPCHasLeader?
  (bind _npc_leader (Variables|NPCLeader|GetNPC%20Leader)
    (:then
      (return true _npc_leader))
    (:else
      (return false 0))))

(fn NPCPawnDebugMessage
  (params (String :type String) (OnlyWhenDebug :type Boolean :default true))
  (if (or (not OnlyWhenDebug) (and OnlyWhenDebug (Variables|NPCDebugPawn|GetNPC%20DebugPawn)))
    (Development|PrintString (Utilities|String|Append "===== " "NPC Pawn | " String " | " (Utilities|GetObjectName self) " ====="))))

(fn NPCPickupDropped
  (params (FortPickupDropped :type Fort%20Pickup%20Object%20Reference))
  (CallFunction|NPCSetReturnGatheredLoottoBlackboard))

(fn NPCPickupGrabbed
  (params (FortPickupGrabbed :type Fort%20Pickup%20Object%20Reference))
  (CallFunction|NPCSetReturnGatheredLoottoBlackboard)
  (if (not (Variables|NPCInventory|GetNPC%20ShowInventoryRarityVisuals))
    (CallFunction|NPCSendGameplayCuePickupGrabbed)))

(fn NPCPickupGrabbed_Destroyed
  (params (ActorGrabbed :type Actor%20Object%20Reference)))

(fn NPCResetOriginalGravityScale
  (Variables|/Script/Engine.CharacterMovementComponent|SetGravityScale (Variables|Internal|GetNPC%20OriginalGravityScale) (Variables|Character|GetCharacterMovement)))

(fn NPCResetOriginalRotationRate
  (Variables|/Script/Engine.CharacterMovementComponent|SetRotationRate (Variables|Internal|GetNPC%20OriginalRotationRate) (Variables|Character|GetCharacterMovement)))

(fn NPCScalableFloatIsValid
  (params (ScalableFloat :type Scalable%20Float%20Structure%20%28by%20ref%29))
  (bind (_value _curve) (Utilities|Struct|BreakScalableFloat ScalableFloat))
  (bind (_curvetable _rowname) (Utilities|Struct|BreakCurveTableRowHandle _curve))
  (return (and (Utilities|IsValid _curvetable :!class /Script/Engine.KismetSystemLibrary) (Utilities|Name|NotEqual(Name) _rowname "None"))))

(fn NPCSendGameplayCuePickupGrabbed
  (bind _gameplaycue_pickupgrabbed (Variables|NPCInventory|GetGameplayCue_PickupGrabbed))
  (Networking|CanExecuteCosmeticEvents
    (:True
      (if (GameplayTags|IsGameplayTagValid _gameplaycue_pickupgrabbed)
        (GameplayCue|ExecuteGameplayCue%28Burst%29Local self (GameplayTags|MakeLiteralGameplayTag _gameplaycue_pickupgrabbed) (Ability|GameplayCue|MakeGameplayCueParameters 0.0))))))

(fn NPCSetBestLootRarityInInventory
  (bind _npc_bestlootrarityinventoryfound (LocalVariables|GetNPC%20BestLootRarityInventoryFound))
  (Networking|SwitchHasAuthority
    (:Authority
      (if (Variables|Inventory|GetbHasInventory)
        (LocalVariables|SetNPC%20BestLootRarityInventoryFound "Common") ; Set low and we increase as we find better things.
        (bind _items (InventoryManagement|GetItemsfromInventoryOwner self))
        (CallFunction|NPCPawnDebugMessage "LUTE LIST")
        (for _array_element _items
          (CallFunction|NPCPawnDebugMessage (Utilities|String|Append "" "LUTE ITAM = " (Utilities|String|ToString(Text) (Item|GetDisplayName _array_element)) " | " "RAR = " (Utilities|String|ToString(Name) (Utilities|Name|EnumtoName (Item|GetRarity _array_element))) ""))
          (if (Math|Byte|Byte>Byte (Item|GetRarity _array_element) (LocalVariables|GetNPC%20BestLootRarityInventoryFound))
            (LocalVariables|SetNPC%20BestLootRarityInventoryFound (Item|GetRarity _array_element) :!pinned true) ; Set what our highest rarity is.... so far...
          ))
        (Utilities|FlowControl|Sequence
          (:then_0
            (if (Math|Byte|Byte>Byte _npc_bestlootrarityinventoryfound (Variables|NPCInventory|GetNPC%20BestLootRarityInInventory))
              (Variables|NPCInventory|SetNPC%20BestLootRarityInInventory _npc_bestlootrarityinventoryfound :!pinned true) ; We found a higher tier item. Set it!
            ))
          (:then_1
            (CallFunction|NPCPawnDebugMessage (Utilities|String|BuildString(Name) "" "BEST LUTE = " (Utilities|Name|EnumtoName (Variables|NPCInventory|GetNPC%20BestLootRarityInInventory)))))
          (:then_2
            (return (Variables|NPCInventory|GetNPC%20BestLootRarityInInventory))))))))

(fn NPCSetCanDropLoot
  (params (CanDropLoot :type Boolean))
  (Variables|NPCInventory|SetNPC_CanDropLoot CanDropLoot))

(fn NPCSetLeader
  (params (Leader :type Actor%20Object%20Reference))
  (bind _self self)
  (bind _returnvalue (AI|GetBlackboard _self))
  (bind (_hasleader _leader) (CallFunction|NPCHasLeader?))
  (bind _output_get (Variables|NPCLeader|SetNPC%20Leader Leader)) ; We don't really care about Validating this here.\n\nYou can set it blank to clear Leader status and behaviors.
  (Utilities|IsValid _self
    (:Is%20Valid
      (Utilities|IsValid _returnvalue
        (:Is%20Valid
          (AI|Components|Blackboard|SetValueasObject _returnvalue (Utilities|Name|MakeLiteralName "Leader") _output_get))
        (:Is%20Not%20Valid
          _)))
    (:Is%20Not%20Valid
      _))
  (return _hasleader _leader))

(fn NPCSetLootCollectionPhaseCompleteToBlackboard
  (params (LootCollectionPhaseComplete :type Boolean))
  (bind _self self)
  (bind _returnvalue (AI|GetBlackboard _self))
  (bind _returnvalue_2 (Utilities|Name|MakeLiteralName "LootCollectionPhaseComplete"))
  (if (Math|Boolean|NotEqual(Boolean) LootCollectionPhaseComplete (AI|Components|Blackboard|GetValueasBool _returnvalue _returnvalue_2))
    (AI|Components|Blackboard|SetValueasBool _returnvalue _returnvalue_2 LootCollectionPhaseComplete)))

(fn NPCSetReturnGatheredLootToBlackboard
  (bind _self self)
  (bind _returnvalue (AI|GetBlackboard _self))
  (bind _returnvalue_2 (Utilities|Name|MakeLiteralName "ReturnGatheredLoot"))
  (bind (_hasitemininventory _totalnumberofinventoryitems _lastinventoryitemindex _hasenoughitemstodeliver) (Default|NPCGetInventoryInformation))
  (Utilities|FlowControl|Sequence
    (:then_0
      (if (Math|Boolean|NotEqual(Boolean) _hasenoughitemstodeliver (AI|Components|Blackboard|GetValueasBool _returnvalue _returnvalue_2))
        (AI|Components|Blackboard|SetValueasBool _returnvalue _returnvalue_2 _hasenoughitemstodeliver)))
    (:then_1
      (if (Math|Boolean|NotEqual(Boolean) _hasitemininventory (Variables|NPCInventory|GetNPC%20HasItemInInventory))
        (Variables|NPCInventory|SetNPC%20HasItemInInventory _hasitemininventory) ; For anim poses
      ))))

(fn NPCSpawnInvisibleBeginsMakeInvisible
  (bind _canhideorunhide (CallFunction|NPCCanHideorUnHidePawnonSpawn))
  ; Do this whole thing in a function so children can add other things to hide (FX, etc)
  (Utilities|FlowControl|Sequence
    (:then_0
      (if _canhideorunhide
        (Rendering|SetActorHiddenInGame true) ; Start hidden and let spawn ability enable visibility
      ))
    (:then_1
      (return _canhideorunhide))))

(fn NPCSpawnInvisibleCompleteMakeVisible
  (bind _canhideorunhide (CallFunction|NPCCanHideorUnHidePawnonSpawn))
  ; Do this whole thing in a function so children can add other things to hide (FX, etc)
  (Utilities|FlowControl|Sequence
    (:then_0
      (if _canhideorunhide
        (Rendering|SetActorHiddenInGame false) ; Un-hide!
      ))
    (:then_1
      (return _canhideorunhide))))

(fn NPCTargetSlotInfo
  (params (TargetSlotType :type Enum_NPC_TargetSlots%20Enum :default "NewEnumerator0") (TargetWeWantToSlotOnto :type Actor%20Object%20Reference))
  (bind _returnvalue (Ability|GetAbilitySystemComponent TargetWeWantToSlotOnto))
  (bind _npc_targetslotcountingge (LocalVariables|GetNPC%20TargetSlotCountingGE))
  (Utilities|IsValid _returnvalue
    (:Is%20Valid
      ;[1 1 1] Dig up the info we care about for the TargetSlotType we are.
      (bind _array_element (Utilities|Array|ForEachLoopwithBreak (Variables|NPCTargetSlots|GetNPC%20TargetSlotsHF) :!label _label)
        (:LoopBody
          (if (Utilities|Enum|Equal(Enum) TargetSlotType (Utilities|Struct|BreakStructNPCTargetSlots _array_element))
            (Utilities|FlowControl|Sequence
              (:then_0
                (LocalVariables|SetNPC%20TargetSlotCountingGE (.TargetSlotCountingGE_7_F6007B01441781AB71302C91A1C4CFD2 (Utilities|Struct|BreakStructNPCTargetSlots _array_element)))
                (LocalVariables|SetNPC%20AI%20TargetSlotSizeHF (.AITargetSlotSizeHF_8_E81162534B067D58D70AE9AC8B4B1F98 (Utilities|Struct|BreakStructNPCTargetSlots _array_element)))
                (LocalVariables|SetNPC%20AI%20TargetSlotDurationHF (.AITargetSlotDurationHF_14_A9B6ED4D4072D7BB507AA5A14118FA5B (Utilities|Struct|BreakStructNPCTargetSlots _array_element)))
                (LocalVariables|SetNPC%20TargetSlotOverrideGE (.TargetSlotOverrideGE_11_6C40525043B50B2B696580B362A65DB3 (Utilities|Struct|BreakStructNPCTargetSlots _array_element)))
                (LocalVariables|SetNPC%20TargetSlotOwnerGE (.TargetSlotOwnerGE_17_19F50D8848750D3BC118D9BBD861CAA1 (Utilities|Struct|BreakStructNPCTargetSlots _array_element)))
                (LocalVariables|SetNPC%20TargetSlotCountingMaxedGE (.TargetSlotCountingMaxedGE_19_BF1E1EF24CCA642F87D4379BB916F271 (Utilities|Struct|BreakStructNPCTargetSlots _array_element))))
              (:then_1
                (goto _label Break)))))
        (:Completed
          (if (Utilities|IsValidClass _npc_targetslotcountingge)
            ;[1 1 1] Return the data for our TargetSlotType and whether or not we can slot, are full, etc.
            (bind _finaleffectcount (Class|AbilityFunctions|GetTotalGameplayEffectCount (Utilities|Array|MakeArray _npc_targetslotcountingge) _returnvalue)) ; How many slotting stacks are on this Target we would like to attack?
            (bind (_asbool _asintrounded) (CallFunction|NPCGetScalableFloatHotfixValue (LocalVariables|GetNPC%20AI%20TargetSlotSizeHF))) ; How many slotting stacks does this NPC AI require to be empty on the Target?
            (bind _finaleffectcount_2 (Class|AbilityFunctions|GetTotalGameplayEffectCount (Utilities|Array|MakeArray (LocalVariables|GetNPC%20TargetSlotOwnerGE)) _returnvalue)) ; Are we already a Slot Owner?\n\nWill help us ignore erroneous CanActivate for ticking BT Services trying to apply too many slot counting GEs on the target.\\
            (return _finaleffectcount 100 _asintrounded (>= _finaleffectcount 100) (== _finaleffectcount 100) (> _finaleffectcount_2 0) (<= _asintrounded (- 100 _finaleffectcount)) TargetSlotType (LocalVariables|GetNPC%20TargetSlotCountingGE) (LocalVariables|GetNPC%20AI%20TargetSlotSizeHF) (LocalVariables|GetNPC%20AI%20TargetSlotDurationHF) (LocalVariables|GetNPC%20TargetSlotOverrideGE) (LocalVariables|GetNPC%20TargetSlotOwnerGE) (LocalVariables|GetNPC%20TargetSlotCountingMaxedGE))
            (else
              _)))))
    (:Is%20Not%20Valid
      _))
  (return -1 -1 -1 false false false false TargetSlotType 0 0 0 0 0 0))

(fn NPCTargetSlotIsUsingOverride
  (params (TargetSlotType :type Enum_NPC_TargetSlots%20Enum :default "NewEnumerator0") (TargetWeWantToSlotOnto :type Actor%20Object%20Reference))
  (bind _self self)
  (bind _returnvalue (Ability|GetAbilitySystemComponent _self))
  (bind (_targetslotcount _targetslotmax _npc_targetslotsrequired _targetslotsmaxormore _targetslotcountisexactlymax _npc_istargetslotowner _npc_hasspacetoslotontarget _targetslottypeout _targetslotcountingge _ai_targetslotsizehf _ai_targetslotdurationhf _targetslotoverridege) (CallFunction|NPCTargetSlotInfo TargetSlotType TargetWeWantToSlotOnto))
  ;[1 1 1] Return whether or not we have been given an Override Slot to attack a Target (IE Target got too close or damaged me)
  (Utilities|IsValid _returnvalue
    (:Is%20Valid
      (if (Utilities|IsValidClass _targetslotoverridege)
        (bind _finaleffectcount (Class|AbilityFunctions|GetTotalGameplayEffectCount (Utilities|Array|MakeArray _targetslotoverridege) _returnvalue))
        (return (!= _finaleffectcount 0))
        (else
          _)))
    (:Is%20Not%20Valid
      _))
  (return false))

(fn NPCTryToDropLootOnDeath
  (params (DamageTags :type Gameplay%20Tag%20Container%20Structure))
  (bind _candroplootbool (LocalVariables|GetCanDropLootBool))
  (LocalVariables|SetCanDropLootBool (Math|Boolean|ANDBoolean (Variables|NPCInventory|GetNPC_CanDropLoot) (or (and (GameplayTags|NotEqual(GameplayTagContainer) (Variables|NPCInventory|GetNPC%20DeathDropLootRequiredDamageTags) (GameplayTags|MakeLiteralGameplayTagContainer)) (GameplayTags|HasAnyTags DamageTags (Variables|NPCInventory|GetNPC%20DeathDropLootRequiredDamageTags))) (GameplayTags|Equal(GameplayTagContainer) (Variables|NPCInventory|GetNPC%20DeathDropLootRequiredDamageTags) (GameplayTags|MakeLiteralGameplayTagContainer))) (or (and (GameplayTags|NotEqual(GameplayTagContainer) (Variables|NPCInventory|GetNPC%20DeathDropLootForbiddenDamageTags) (GameplayTags|MakeLiteralGameplayTagContainer)) (not (GameplayTags|HasAnyTags DamageTags (Variables|NPCInventory|GetNPC%20DeathDropLootForbiddenDamageTags)))) (GameplayTags|Equal(GameplayTagContainer) (Variables|NPCInventory|GetNPC%20DeathDropLootForbiddenDamageTags) (GameplayTags|MakeLiteralGameplayTagContainer))) (Utilities|Name|NotEqual(Name) (Variables|NPCInventory|GetNPC%20DeathDropLootTierGroupName) "None"))) ; Are Required and Forbidden tags obeyed?
  (if (Variables|NPCInventory|GetNPC_IsInventoryInitialized)
    (Utilities|FlowControl|Sequence
      (:then_0
        (if _candroplootbool
          (bind _outloottodrop (Loot|PickLootDrops (Variables|NPCInventory|GetNPC%20DeathDropLootTierGroupName) (Math|Integer|SelectInt 0 (Loot|GetLootLevel) (FortGlobals|IsinAthena)) -1 "(GameplayTags=)"))
          (for _array_element _outloottodrop
            (Inventory|GiveItemEntrytoInventoryOwner self _array_element) ; Give items from this Loot Group to the Pawn
          )
          (Inventory|DropAllItems true 0) ; Immediately drop the items
        ))
      (:then_1
        (return _candroplootbool)))))

(fn NPCUpdateInteract
  (params (CanInteract :type Boolean))
  (bind _capsulecomponent (Variables|Character|GetCapsuleComponent))
  (Utilities|FlowControl|Sequence
    (:then_0
      (if CanInteract
        (bind _asbool (CallFunction|NPCGetScalableFloatHotfixValue (Variables|NPCInteraction|GetNPC%20AllowInteractWhenHasLeaderHF)))
        (bind _output_get (LocalVariables|SetwasInteractEnabled (Math|Boolean|ORBoolean _asbool (not (CallFunction|NPCHasLeader?)) (Variables|NPCInteraction|GetNPC%20TempInteractOverride))))
        (if _output_get
          (AI|SetCanInteract true)
          (Collision|SetCollisionResponsetoChannel _capsulecomponent "ECC_GameTraceChannel1" "ECR_Overlap")
          (else
            _)))
      (AI|SetCanInteract false)
      (Collision|SetCollisionResponsetoChannel _capsulecomponent "ECC_GameTraceChannel1" "ECR_Ignore"))
    (:then_1
      (return (LocalVariables|GetwasInteractEnabled)))))

(fn OnRep_NPCBestLootRarityInInventory
  (Networking|CanExecuteCosmeticEvents
    (:True
      (if (Variables|NPCInventory|GetNPC%20ShowInventoryRarityVisuals)
        (CallFunction|NPCSendGameplayCuePickupGrabbed)))))

(fn OnRep_NPCDeathTime
  (Networking|CanExecuteCosmeticEvents
    (:True
      (CallFunction|NPCUpdateHealthBarComponentVisibility))))

(fn OnRep_NPCInvisiblePortionOfSpawningCompletedMakeVisible
  ; Placed in OnRep so Clients will update the visibility to true when the pawn becomes relevant
  ;
  ; Notes:
  ; -This is False by default
  ; -This is only ever set to True in cases where we spawn invisibly or scale-in spawn (scale-in also uses invisibility)
  (if (Variables|NPCSpawning|GetNPC%20InvisiblePortionOfSpawningCompletedMakeVisible)
    (CallFunction|NPCSpawnInvisibleCompleteMakeVisible)))

(fn OnRep_NPCLeader
  (bind _npc_skeletal_mesh_index (Variables|Internal|GetNPC%20Skeletal%20Mesh%20Index))
  (bind _returnvalue (Ability|GameplayCue|MakeGameplayCueParameters 0.0 _npc_skeletal_mesh_index))
  (bind _self self)
  (bind _gameplaycue_leaderset (Variables|NPCLeader|GetGameplayCue_LeaderSet))
  (if (GameplayTags|IsGameplayTagValid (Variables|NPCLeader|GetGameplayCue_LeaderSet))
    (if (Variables|NPCLeader|GetisGCForLeaderSetLooping)
      (Utilities|IsValid (Variables|NPCLeader|GetNPC%20Leader)
        (:Is%20Valid
          (Projectile|AddGameplayCue%28Looping%29Local _self _gameplaycue_leaderset _returnvalue))
        (:Is%20Not%20Valid
          (Projectile|RemoveGameplayCue%28Looping%29Local _self _gameplaycue_leaderset _returnvalue))) ; Reset when leader is unset
    )))

(fn OnRep_NPC_AlertLevel
  (bind (_alertlevel _alertleveltag _validdatafound) (CallFunction|NPCGetAlertLevelInfo))
  (bind (_alertlevel_2 _alertleveltag_2 _validdatafound_2) (CallFunction|NPCGetPreviousAlertLevelInfo))
  (Networking|SwitchHasAuthority
    (:Authority
      (Default|CallNPCAlertLevelUpdated (Variables|NPCAlertLevel|GetNPC_AlertLevel)))
    (:Remote
      (CallFunction|NPCAlertLevelUpdatedClient)))
  ;[1 1 1] Apply AlertLevel Gameplay Tags
  (Utilities|FlowControl|Sequence
    (:then_0
      (if _validdatafound_2
        (Tags|RemoveNonReplicatedTagtoPawn _alertleveltag_2)))
    (:then_1
      (if _validdatafound
        (Tags|AddNonReplicatedTagtoPawn _alertleveltag)))))

(fn OnRep_SelectedPrimitiveDataFloat
  (CallFunction|SetPrimitiveDataFloat))

(fn Orphaned
  (return false 0))

(fn PlayWaterSplashatWaterSurface
  (params (Water%20Body :type Fort%20Water%20Body%20Actor%20Object%20Reference) (GameplayCueTag :type Gameplay%20Tag%20Structure))
  (bind _outwatersurfaceinfo (WaterBody|GetWaterSurfaceInfoStruct Water%20Body (Transformation|GetActorLocation) false true false))
  (GameplayCue|ExecuteGameplayCue%28Burst%29Local self (Variables|GetGameplayCueTag) (Ability|GameplayCue|MakeGameplayCueParameters :NormalizedMagnitude (Math|Conversions|ToFloat(Integer) (.WaterBodyIdx (Utilities|Struct|BreakWaterSurfaceInfo _outwatersurfaceinfo))) :RawMagnitude (.WaterDepth (Utilities|Struct|BreakWaterSurfaceInfo _outwatersurfaceinfo)) :Location (.WaterSurfaceLocation (Utilities|Struct|BreakWaterSurfaceInfo _outwatersurfaceinfo)) :Normal (.WaterPlaneNormal (Utilities|Struct|BreakWaterSurfaceInfo _outwatersurfaceinfo)) :GameplayEffectLevel 1 :AbilityLevel 1 :bReplicateLocationWhenUsingMinimalRepProxy false)))

(fn PlayAdditiveHitReacts
  (params (Hit%20Direction :type Vector%20%28by%20ref%29) (Anim%20Montage :type Anim%20Montage%20Object%20Reference))
  (HitReacts|PlayAdditiveHitReactsNative Hit%20Direction Anim%20Montage) ; This used to be done directly in the Event graph, which meant any Child of this Pawn couldn't override the Function behavior.\n\nWith this, we should be able to override PlayAdditiveHitReacts and call the Native function here with different logic in front of it, if we need to.
)

(fn PlayStateChangeAudio
  (params (Index :type Enum_NPC_AlertLevel%20Enum :default "NewEnumerator4")))

(fn RemoveLibraryCollection
  (bind _sound_library_collection_tags (Variables|SoundLibraries|GetSound%20Library%20Collection%20Tags))
  (if (Utilities|GameplayTagContainerNotEmpty _sound_library_collection_tags)
    (SoundLibrary|RemoveLibrariesfromCollections (WorldSubsystems|GetSoundLibrarySubsystem) self _sound_library_collection_tags)))

(fn ResetMaterialFromDeath
  (bind _mesh (Variables|Character|GetMesh)
    (:then
      (LocalVariables|SetNumMaterials (Utilities|Array|Length (Rendering|Material|GetMaterials _mesh)))
      (for (_array_element _array_index) (Variables|NPCDeath|GetOG%20Materials)
        (Utilities|IsValid _array_element
          (:Is%20Valid
            (if (< _array_index (LocalVariables|GetNumMaterials))
              (bind _mesh_2 (Variables|Character|GetMesh)
                (:then
                  (Rendering|Material|SetMaterial _mesh_2 _array_index _array_element))
                (:else))))))
      (return))
    (:else
      (return))))

(fn SelectPrimitiveDataFloat
  (bind _asbool (CallFunction|NPCGetScalableFloatHotfixValue (Variables|PrimitiveDataFloat|GetNPC%20PrimtiveDataFloatSetRandomlyInRange)))
  (if _asbool
    (bind (_asbool_2 _asintrounded _asfloat) (CallFunction|NPCGetScalableFloatHotfixValue (Variables|PrimitiveDataFloat|GetNPC%20PrimtiveDataFloatMinRangeHF)))
    (bind (_asbool_3 _asintrounded_2 _asfloat_2) (CallFunction|NPCGetScalableFloatHotfixValue (Variables|PrimitiveDataFloat|GetNPC%20PrimtiveDataFloatMaxRangeHF)))
    (Variables|PrimitiveDataFloat|SetSelectedPrimitiveDataFloat (Math|Random|RandomFloatinRange _asfloat _asfloat_2))))

(fn SetPrimitiveDataFloat
  (bind (_asbool _asintrounded) (CallFunction|NPCGetScalableFloatHotfixValue (Variables|PrimitiveDataFloat|GetNPC%20PrimtiveDataFloatIndexHF)))
  (Rendering|Material|SetCustomPrimitiveDataFloat (Variables|Character|GetMesh) _asintrounded (Variables|PrimitiveDataFloat|GetSelectedPrimitiveDataFloat)))

(fn ShouldRunInstantDeathCue
  (params (DeathTags :type Gameplay%20Tag%20Container%20Structure) (Instigator :type Actor%20Object%20Reference) (DamageCauser :type Actor%20Object%20Reference))
  (return (Math|Boolean|ANDBoolean (or (GameplayTags|HasTag DeathTags (Variables|Default|GetGCInstantDeathTag)) (GameplayTags|HasTag DeathTags (Variables|Default|GetGCInstantDeathTagTwo))) (GameplayTags|NotEqual(GameplayTag) :A (Variables|NPCDeath|GetGameplayCue_SystemInstantDeath)) (or (not (Utilities|IsValid Instigator :!class /Script/Engine.KismetSystemLibrary)) (== Instigator (Pawn|GetController))) (not (Utilities|IsValid DamageCauser :!class /Script/Engine.KismetSystemLibrary)))))

(fn ShouldUpdateSkeletalMesh
  ;[1 1 1] HACK
  (return true))

(fn StoreOriginalGravityAndRotation
  (bind _charactermovement (Variables|Character|GetCharacterMovement))
  (Variables|Internal|SetNPC%20OriginalRotationRate (Variables|/Script/Engine.CharacterMovementComponent|GetRotationRate _charactermovement))
  (Variables|Internal|SetNPC%20OriginalGravityScale (Variables|/Script/Engine.CharacterMovementComponent|GetGravityScale _charactermovement)))

(fn TryShowHealthBar
  (bind _array_element (Utilities|Array|ForEachLoopwithBreak (Fortnite|GetLocalFortPlayerControllers true) :!label _label)
    (:LoopBody
      (bind (_asbool _asintrounded _asfloat) (CallFunction|NPCGetScalableFloatHotfixValue (Variables|NPCStatusWidgetHealthBar|GetNPC%20MaxDistanceFromDamagedNPCToDrawHeallthBarHF)))
      (if (Math|Boolean|ANDBoolean (< (Class|AbilityFunctions|GetDistanceBetweenActors (Pawn|GetControlledPawn _array_element) self) _asfloat) (not (Variables|Health|GetbIsDying)) (not (Variables|NPCStatusWidgetHealthBar|GetNPC%20StatusWidgetHealthBarDisabled)))
        (bind _npcstatuswidget (Variables|Default|GetNPCStatusWidget)
          (:then
            (bind _asnpcstatus_widget (Utilities|Casting|CastToNPCStatusWidget (UserInterface|GetUserWidgetObject _npcstatuswidget)))
            (Class|NPCStatusWidget|EventUpdateStatus _asnpcstatus_widget (/ (FortPawn|GetHealth) (FortPawn|GetMaxHealth)) (/ (FortPawn|GetShield) (FortPawn|GetMaxShield)) (> (FortPawn|GetMaxShield) 0))
            (goto _label Break))
          (:else))))
    (:Completed
      (return))))

(fn UpdateSoundLibrary
  (params (Alert%20Level :type Enum_NPC_AlertLevel%20Enum :default "NewEnumerator4"))
  (Utilities|IsValid (WorldSubsystems|GetSoundLibrarySubsystem)
    (:Is%20Valid
      (Sound|PlayStateChangeAudio Alert%20Level)
      (switch Utilities|FlowControl|Switch|SwitchonEnum_NPC_AlertLevel Alert%20Level
        (:NewEnumerator1
          (if (Sound|CanUpdateSoundLib (Variables|SoundLibraries|GetSoundLibs_Default))
            (Sound|CheckSoundLibraryRemoved (Variables|SoundLibraries|GetSoundLibs_Alert))
            (Sound|CheckSoundLibraryRemoved (Variables|SoundLibraries|GetSoundLibs_Aggressive))
            (Sound|CheckSoundLibraryAdded (Variables|SoundLibraries|GetSoundLibs_Default))))
        (:NewEnumerator2
          (if (Utilities|IsValid (Variables|Sound|GetCurrentSoundLib) :!class /Script/Engine.KismetSystemLibrary)
            (if (Sound|CanUpdateSoundLib (Variables|SoundLibraries|GetSoundLibs_Alert))
              (Sound|CheckSoundLibraryRemoved (Variables|SoundLibraries|GetSoundLibs_Default))
              (Sound|CheckSoundLibraryRemoved (Variables|SoundLibraries|GetSoundLibs_Aggressive))
              (Sound|CheckSoundLibraryAdded (Variables|SoundLibraries|GetSoundLibs_Alert)))))
        (:NewEnumerator3
          (if (Utilities|IsValid (Variables|Sound|GetCurrentSoundLib) :!class /Script/Engine.KismetSystemLibrary)
            (if (Sound|CanUpdateSoundLib (Variables|SoundLibraries|GetSoundLibs_Aggressive))
              (Sound|CheckSoundLibraryRemoved (Variables|SoundLibraries|GetSoundLibs_Default))
              (Sound|CheckSoundLibraryAdded (Variables|SoundLibraries|GetSoundLibs_Aggressive))
              (Sound|CheckSoundLibraryRemoved (Variables|SoundLibraries|GetSoundLibs_Alert)))))))))

(fn UpdateSoundLibraryCollection
  (params (AddLibraries :type Boolean))
  (bind _alertlevel (CallFunction|NPCGetAlertLevelInfo))
  (if (Utilities|GameplayTagContainerNotEmpty (Variables|SoundLibraries|GetSound%20Library%20Default%20Tags))
    (if (Variables|GetAddLibraries)
      (if (Utilities|Enum|Equal(Enum) _alertlevel "NewEnumerator2")
        (if (not (or (== (Variables|SoundLibraries|GetSound%20Library%20Collection%20Tags) (Variables|SoundLibraries|GetSound%20Library%20Alert%20Tags)) (SoundLibraries|CheckTag (Variables|SoundLibraries|GetSound%20Library%20Alert%20Tags))))
          (SoundLibraries|RemoveLibraryCollection)
          (Variables|SoundLibraries|SetSound%20Library%20Collection%20Tags (Variables|SoundLibraries|GetSound%20Library%20Alert%20Tags))
          (SoundLibrary|AddLibrariesfromCollections (WorldSubsystems|GetSoundLibrarySubsystem) self (Variables|SoundLibraries|GetSound%20Library%20Collection%20Tags) :!label _label)
          (return true))
        (elif (Utilities|Enum|Equal(Enum) _alertlevel "NewEnumerator3")
          (if (not (or (== (Variables|SoundLibraries|GetSound%20Library%20Collection%20Tags) (Variables|SoundLibraries|GetSound%20Library%20Aggressive%20Tags)) (SoundLibraries|CheckTag (Variables|SoundLibraries|GetSound%20Library%20Aggressive%20Tags))))
            (SoundLibraries|RemoveLibraryCollection)
            (Variables|SoundLibraries|SetSound%20Library%20Collection%20Tags (Variables|SoundLibraries|GetSound%20Library%20Aggressive%20Tags))
            (goto _label))
          (elif (== (Variables|SoundLibraries|GetSound%20Library%20Collection%20Tags) (Variables|SoundLibraries|GetSound%20Library%20Default%20Tags))
            (return true)
            (else
              (SoundLibraries|RemoveLibraryCollection)
              (Variables|SoundLibraries|SetSound%20Library%20Collection%20Tags (Variables|SoundLibraries|GetSound%20Library%20Default%20Tags))
              (goto _label)))))
      (return true)
      (else
        (SoundLibraries|RemoveLibraryCollection)))
    (else
      (return false))))

(fn UpdateTrackedActorStateFromCurrentAlertLevel
  (bind (_alertlevel _alertleveltag _validdatafound) (CallFunction|NPCGetAlertLevelInfo))
  (bind _returnvalue (FortGameplayDataTrackerComponent|GetFortGameplayDataTrackerComponentManager))
  (Utilities|IsValid _returnvalue
    (:Is%20Valid
      (if _validdatafound
        (Class|FortGameplayDataTrackerComponentManager|SetTrackedActorState _returnvalue self _alertleveltag)))))

(fn ConstructionScript
  (Utilities|FlowControl|Sequence
    (:then_0
      (Networking|CanExecuteCosmeticEvents
        (:True
          (bind _weaponcapsulecollision (Variables|Default|GetWeaponCapsuleCollision)
            (:then
              (AI|SetWeaponCollisionComponent _weaponcapsulecollision) ; For simplified/smoothed weapon collision that is attached to the mesh without using a full-on physics body setup.\n\nCan also be used in combination where you use the weapon capsule, and add phys bodies for things that fall outside of the capsule.
            )
            (:else)))))
    (:then_1
      ; We also do this OnBeginPlay, but we get visual pops that way sometimes.
      (if (not (Variables|NPCSpawning|GetNPC%20InvisiblePortionOfSpawningCompletedMakeVisible))
        (CallFunction|NPCSpawnInvisibleBeginsMakeInvisible) ; Make pawn Invisible! (if we are allowed to)
      ))))
