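; Parent pawn Blueprint for Fortnite wildlife NPCs (Irwin), adding taming, status UI, audio and cosmetics.
; Convert component events set the leader and team, swap gameplay effects, and drive interaction rules and team-only markers and health bars.
; Sound Library calls play creature audio, surface checks orient foot-dust FX, and animation layers async-load on clients unless frame stripping is on.
; Launch counting breaks stuck bouncing with randomly rotated forward velocity.
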(event EventEndPlay
  (params (EndPlayReason :type EEndPlayReason%20Enum :default "Destroyed"))
  (|CallParentFunction EndPlayReason)
  (Customization|UnbindallEventsfromOnFinishedCustomization (Variables|Default|GetFortCharacterParts)))

(event Custom|On%20Customization%20Applied
  (params (Value :type Custom%20Character%20Part%20Owner%20Interface%20Interface))
  ;[1 1 1] Set color variation on the AI once all parts are applied
  (CallFunction|ApplyVariation))

(event Custom|CE_UpdateSurfaceType
  (params (Component :type Actor%20Component%20Object%20Reference) (bReset :type Boolean))
  (bind _returnvalue (Footsteps|GetSurfaceTrackingComponent))
  (bind _returnvalue_2 (Class|FortSurfaceTrackingComponent|GetCurrentSurfaceType _returnvalue))
  (bind _returnvalue_3 (Utilities|Enum|Equal(Enum) _returnvalue_2 "SurfaceType30"))
  (bind _wildlife_footdust (Variables|Default|GetWildlife_FootDust))
  ;[1 1 1] WIP mud FX pass
  (Utilities|FlowControl|Sequence
    (:then_0
      (Niagara|SetNiagaraVariable(Bool) _wildlife_footdust "IsMud" _returnvalue_3)
      (bind _isa (Utilities|FlowControl|FlipFlop)
        (:A
          (Niagara|SetNiagaraVariable(Bool) _wildlife_footdust "bIsRight" _isa))
        (:B
          (Niagara|SetNiagaraVariable(Bool) _wildlife_footdust "bIsRight" _isa))))
    (:then_1
      (if _returnvalue_3
        (bind _outhit (Collision|LineTraceForObjects (- (Transformation|GetActorLocation) "0.000000,0.000000,100.000000") (- (Transformation|GetActorLocation) "0.000000,0.000000,300.000000") (Utilities|Array|MakeArray "ObjectTypeQuery1")))
        (Transformation|SetWorldRotation _wildlife_footdust (Math|Rotator|MakeRotfromZX (.ImpactNormal (Collision|BreakHitResult _outhit)) (Transformation|GetActorForwardVector)) false false)))))

(event EventBeginPlay
  (|CallParentFunction)
  (Utilities|FlowControl|Sequence
    (:then_0
      (bind _npcstatuswidget (Variables|Default|GetNPCStatusWidget)
        (:then
          (bind _asnpcstatus_widget (Utilities|Casting|CastToNPCStatusWidget (UserInterface|GetUserWidgetObject _npcstatuswidget)))
          (Variables|/Game/Athena/HUD/Actors/NPCStatusWidget.NPCStatusWidget_C|SetFortPawn self _asnpcstatus_widget))
        (:else)))
    (:then_1
      ;[1 1 1] WIP mud FX pass
      (Utilities|EffectsQualityCheck 0
        (:Cinematic
          _)
        (:Epic
          _)
        (:High
          _))
      (Components|Activation|BindEventtoOnComponentActivated (Variables|Default|GetWildlife_FootDust) (delegate Custom|CE_UpdateSurfaceType)))
    (:then_2
      (Customization|BindEventtoOnFinishedCustomization (Variables|Default|GetFortCharacterParts) (EventDispatchers|CreateEvent :!event Custom|On%20Customization%20Applied)))))

(event Custom|On%20Pawn%20Focused
  (params (FocusedPawn :type Fort%20Pawn%20Object%20Reference))
  (bind _self self)
  (bind _returnvalue (Actor|GetComponentbyClass _self "/Script/FortniteGame.FortPawnComponent_Convert"))
  (bind _returnvalue_2 (Convert|GetConverterPawn _returnvalue))
  ;[1 1 1] Only do this for converted wildlife
  (if (== FocusedPawn self)
    (if (not (Variables|Default|GetbIsBeingRidden))
      (Utilities|IsValid _returnvalue
        (:Is%20Valid
          (Utilities|IsValid _returnvalue_2
            (:Is%20Valid
              (bind _asfort_player_state_athena (Utilities|Casting|CastToFortPlayerStateAthena (Variables|/Script/Engine.Pawn|GetPlayerState _returnvalue_2)))
              (bind _asfort_player_state_athena_2 (Utilities|Casting|CastToFortPlayerStateAthena (Variables|/Script/Engine.Controller|GetPlayerState (Fortnite|GetFirstLocalFortPlayerController true))))
              (if (== (Class|FortPlayerStateAthena|GetSquadId _asfort_player_state_athena) (Class|FortPlayerStateAthena|GetSquadId _asfort_player_state_athena_2))
                (bind _npcstatuswidget (Variables|Default|GetNPCStatusWidget)
                  (:then
                    (bind _asnpcstatus_widget (Utilities|Casting|CastToNPCStatusWidget (UserInterface|GetUserWidgetObject _npcstatuswidget)))
                    (Class|NPCStatusWidget|OnPawnFocused _asnpcstatus_widget)
                    (Class|NPCStatusWidget|EventUpdateStatus _asnpcstatus_widget (/ (FortPawn|GetHealth) (FortPawn|GetMaxHealth)) 0.0 false))
                  (:else))))))))))

(event EventOnCustomizationsLoaded_BP
  (|CallParentFunction)
  ;[1 1 1] Don't link the animation layers on low-memory platforms.
  ; (Low-memory platforms are platforms where we frame-strip animations.)
  (bind _returnvalue (Development|GetConsoleVariableIntValue "a.StripFramesOnCompression"))
  (if (== _returnvalue 0)
    (CallFunction|LinkAnimationLayers)))

(event Collision|EventActorBeginOverlap
  (params (OtherActor :type Actor%20Object%20Reference)))

(event Custom|HideStatus
  (bind _npcstatuswidget (Variables|Default|GetNPCStatusWidget)
    (:then
      (bind _asnpcstatus_widget (Utilities|Casting|CastToNPCStatusWidget (UserInterface|GetUserWidgetObject _npcstatuswidget)))
      (Class|NPCStatusWidget|EventHideStatus _asnpcstatus_widget))
    (:else)))

(event EventNPCAlertLevelChangedServer
  (params (NPCAlertLevel :type Enum_NPC_AlertLevel%20Enum :default "NewEnumerator4"))
  (|CallParentFunction NPCAlertLevel)
  (switch Utilities|FlowControl|Switch|SwitchonEnum_NPC_AlertLevel NPCAlertLevel
    (:NewEnumerator1
      ;[1 1 1] Clear cooldown (infinite duration) for alert when unaware again. Not sure if this is efficient
      (GameplayEffects|RemoveActiveGameplayEffectbySourceEffect (Variables|GameplayAbilities|GetAIPawnAbilitySystemComponent) "/Game/Athena/AI/Abilities/NPC/Parents/GE_NPC_Parent_Alert_Cooldown.GE_NPC_Parent_Alert_Cooldown_C" (Variables|GameplayAbilities|GetAIPawnAbilitySystemComponent) -1))))

(event EventNPCLeaderSet
  (params (Success? :type Boolean) (Leader :type Actor%20Object%20Reference) (UnconvertReason :type EUnconvertReason%20Enum :default "StealByOtherPlayer"))
  ;[1 1 1] Taming Audio
  (|CallParentFunction Success? Leader UnconvertReason)
  (bind _updated_variable (Utilities|FlowControl|DeltaGateWildcard (Variables|Default|GetbLeaderSetSuccess) Success?)
    (:Success
      (if _updated_variable
        _))))

(event EventOnNPCPawnInteractedWith
  (params (AIPawn :type Actor%20Object%20Reference) (InteractingPlayerPawn :type Fort%20Player%20Pawn%20Object%20Reference) (InteractionBeingAttempted :type EInteractionBeingAttempted%20Enum :default "FirstInteraction"))
  ;[1 1 1] Taming
  (|CallParentFunction AIPawn InteractingPlayerPawn "FirstInteraction")
  (if (not (Variables|NPCLeader|GetNPCDynamicBlockConversionOnInteract))
    (bind _asfort_pawn (Utilities|Casting|CastToFortPawn AIPawn))
    (Convert|TriggerConversion InteractingPlayerPawn _asfort_pawn)))

(event OnConvertedEvent(ConvertComponent)
  (params (InstigatorPawn :type Fort%20Pawn%20Object%20Reference) (ConvertedPawn :type Fort%20Pawn%20Object%20Reference))
  (bind _returnvalue (Ability|GetFortAbilitySystemComponent InstigatorPawn))
  (bind _self self)
  (bind _returnvalue_2 (Ability|GetFortAbilitySystemComponent _self))
  ;[1 1 1] Taming
  (CallFunction|NPCAttempttoSetLeader InstigatorPawn "Unknown")
  (Default|CallConvertedDispatcher true InstigatorPawn)
  (AI|SetCanInteract (CallFunction|CanbeInteractedwithwhenTamed))
  (Variables|Default|SetLeaderTeam (Team|GetActorTeam InstigatorPawn))
  (Utilities|FlowControl|Sequence
    (:then_0
      (Networking|CanExecuteCosmeticEvents
        (:True
          (Default|PlaySoundLib self (Variables|Default|GetTameSoundLibTag)))))
    (:then_1
      (Utilities|IsValid _returnvalue_2
        (:Is%20Valid
          (GameplayEffects|ApplyGameplayEffectToSelf :self _returnvalue_2 :GameplayEffectClass (Variables|Default|GetGE_GoalSelectionWhileTamed) :Level 0.0) ; Grant tag to give new goal selection params
        )
        (:Is%20Not%20Valid
          _))
      (Utilities|IsValid _returnvalue
        (:Is%20Valid
          (bind _returnvalue_3 (GameplayEffects|ApplyGameplayEffectToSelf :self _returnvalue :GameplayEffectClass (Variables|Conversion|GetGE_Quest_Converted_Irwin) :Level 0.0)) ; Grant a tag to the leader to denote they've successfully tamed irwin
          (Variables|Conversion|SetGE_Quest_Converted_Handle _returnvalue_3))))))

(event OnUnconvertedEvent(ConvertComponent)
  (params (UnconvertedPawn :type Fort%20Pawn%20Object%20Reference) (UnconvertReason :type EUnconvertReason%20Enum :default "StealByOtherPlayer"))
  (bind _self self)
  (bind _returnvalue (Ability|GetFortAbilitySystemComponent _self))
  ;[1 1 1] Taming
  (CallFunction|NPCAttempttoSetLeader 0 UnconvertReason)
  (Default|CallConvertedDispatcher false)
  (bind _caninteract (CallFunction|NPCCanbeInteractedWith))
  (CallFunction|NPCUpdateInteract _caninteract) ; Test to put interact back on
  (Variables|Default|SetLeaderTeam 0)
  (Utilities|IsValid _returnvalue
    (:Is%20Valid
      (GameplayEffects|RemoveActiveEffectswithTags _returnvalue (Variables|Default|GetT_GoalSelectionWhileTamed)))
    (:Is%20Not%20Valid
      _))
  (Utilities|IsValid (Ability|GetFortAbilitySystemComponent (stale OnConvertedEvent(ConvertComponent) InstigatorPawn))
    (:Is%20Valid
      (GameplayEffects|RemoveActiveGameplayEffect (Ability|GetFortAbilitySystemComponent (stale OnConvertedEvent(ConvertComponent) InstigatorPawn)) (Variables|Conversion|GetGE_Quest_Converted_Handle) -1)
      (CallFunction|RestoreConvertResistance))))

(event EventOnOutOfConvertResistanceServer
  (params (ConvertDamageDone :type Float%20%28single-precision%29) (SourceTags :type Gameplay%20Tag%20Container%20Structure%20%28by%20ref%29) (EffectContext :type Gameplay%20Effect%20Context%20Handle%20Structure%20%28by%20ref%29) (EffectInstigator :type Controller%20Object%20Reference) (EffectCauser :type Actor%20Object%20Reference))
  ;[1 1 1] Taming
  ;;[1 1 1] Convert pawn once it runs out of convert resistance
  (bind _asfort_pawn (Utilities|Casting|CastToFortPawn EffectCauser))
  (Convert|TriggerConversion _asfort_pawn self))

(event Custom|RestoreConvertResistance
  (bind _self self)
  (bind _returnvalue (Ability|GetFortAbilitySystemComponent _self))
  ;[1 1 1] Taming
  ;;[1 1 1] Restore Convert Resistance
  (Networking|SwitchHasAuthority
    (:Authority
      (if (Variables|Convert|GetbSupportsConvertSet)
        (bind _result (Math|Float|NegateFloat (Convert|GetMaxConvertResistance))
          (:%20%20
            (bind _returnvalue_2 (Ability|GameplayEffect|AssignTagSetbyCallerMagnitude (GameplayEffects|MakeOutgoingSpec :self _returnvalue :GameplayEffectClass "/Game/Athena/AI/Abilities/NPC/Parents/GE_NPC_Parent_Convert_Damage.GE_NPC_Parent_Convert_Damage_C" :Level 0.0) "(TagName=\"SetByCaller.ConvertDamage\")" _result))
            (GameplayEffects|ApplyGameplayEffectSpecToSelf _returnvalue _returnvalue_2)))))))

(event EventOnDeathServer
  (params (Damage :type Float%20%28single-precision%29) (DamageTags :type Gameplay%20Tag%20Container%20Structure%20%28by%20ref%29) (Momentum :type Vector :default "0, 0, 0") (HitInfo :type Hit%20Result%20Structure%20%28by%20ref%29) (InstigatedBy :type Controller%20Object%20Reference) (DamageCauser :type Actor%20Object%20Reference) (EffectContext :type Gameplay%20Effect%20Context%20Handle%20Structure))
  (bind _wildlife_footdust (Variables|Default|GetWildlife_FootDust))
  (|CallParentFunction Damage DamageTags Momentum HitInfo InstigatedBy DamageCauser EffectContext)
  (Utilities|EffectsQualityCheck 0
    (:Cinematic
      _)
    (:Epic
      _)
    (:High
      _))
  (Utilities|IsValid _wildlife_footdust
    (:Is%20Valid
      (Components|Activation|UnbindEventfromOnComponentActivated _wildlife_footdust (delegate Custom|CE_UpdateSurfaceType)))))

(event AthenaMovement|EventOnSlidingStateChanged
  (params (bNewSlidingState :type Boolean))
  ;[1 1 1] Sliding GCNL for audio
  (if bNewSlidingState
    (GameplayCue|AddGameplayCue self (Variables|Default|GetIrwinSlidingGC) (Class|AthenaFunctionLibrary|MakeSimpleFortEffectContext self))
    (else
      (GameplayCue|RemoveGameplayCue self (Variables|Default|GetIrwinSlidingGC)))))

(event Custom|LinkAnimationLayers
  ;[1 1 1] Link Animation Layers
  (CallFunction|AsyncLinkAnimationLayer (Variables|NPCAnimation|GetLayer%20Idle))
  (CallFunction|AsyncLinkAnimationLayer (Variables|NPCAnimation|GetLayer%20Locomotion))
  (CallFunction|AsyncLinkAnimationLayer (Variables|NPCAnimation|GetLayer%20Jump)))

(event Custom|AsyncLinkAnimationLayer
  (params (AnimationLayer :type Anim%20Instance%20Soft%20Class%20Reference))
  ;[1 1 1] Async Link Animation Layer
  (Networking|SwitchHasAuthority
    (:Remote
      (if (Utilities|IsValidSoftClassReference AnimationLayer)
        (bind _class (Utilities|AsyncLoadClassAsset AnimationLayer)
          (:Completed
            (bind _asanim_instance (Utilities|Casting|CastToAnimInstanceClass _class))
            (Animation|LinkAnimLayer _asanim_instance)))))))

(event EventOnDeathPlayEffects
  (params (Damage :type Float%20%28single-precision%29) (DamageTags :type Gameplay%20Tag%20Container%20Structure%20%28by%20ref%29) (Momentum :type Vector :default "0, 0, 0") (HitInfo :type Hit%20Result%20Structure%20%28by%20ref%29) (InstigatedBy :type Fort%20Pawn%20Object%20Reference) (DamageCauser :type Actor%20Object%20Reference) (EffectContext :type Gameplay%20Effect%20Context%20Handle%20Structure))
  ;[1 1 1] Death sounds
  (|CallParentFunction Damage DamageTags Momentum HitInfo InstigatedBy DamageCauser EffectContext)
  (Default|PlaySoundLib "(TagName=\"SoundLibrary.Irwin.VO.Death\")"))

(event OnSoundLibraryPlayEvent(SoundLibrary_Component)
  (params (Results :type Sound%20Library%20Play%20Result%20Structure%20%28by%20ref%29) (Context :type Sound%20Library%20Context%20Object%20Reference))
  (CallFunction|AttachandApplyVocalFX Results))

(event EventOnLanded
  (params (Hit :type Hit%20Result%20Structure%20%28by%20ref%29))
  (Networking|CanExecuteCosmeticEvents
    (:True
      (Default|PlaySoundLib self (Variables|Default|GetJumpLandSoundLibTag)))))

(event Controlled|EventOnBeingControlledChanged
  (params (bControlled :type Boolean))
  (bind _convertcomponent (Variables|Default|GetConvertComponent))
  ;[1 1 1] Indicator Marker
  (Variables|/Script/FortniteGame.FortPawnComponent_Convert|SetbCanShowMarker (not bControlled) _convertcomponent)
  (if bControlled
    (Convert|HideIndicatorMarker _convertcomponent)
    (MiniMap|HideMapMarker)
    (CallFunction|NPCStatusWidgetUpdateKeepVisible false)
    (else
      (Default|UpdateTamingIndicator))))

(event EventOnApplyEffectOptimization
  (params (NewBudgetValue :type Integer))
  (if (not (FortPawn|IsDead))
    (Rendering|SetVisibility (Variables|Default|GetWildlife_FootDust) (<= NewBudgetValue 1) true)))

(event EventOnDamagePlayEffects
  (params (Damage :type Float%20%28single-precision%29) (DamageTags :type Gameplay%20Tag%20Container%20Structure%20%28by%20ref%29) (Momentum :type Vector :default "0, 0, 0") (HitInfo :type Hit%20Result%20Structure%20%28by%20ref%29) (InstigatedBy :type Fort%20Pawn%20Object%20Reference) (DamageCauser :type Actor%20Object%20Reference) (EffectContext :type Gameplay%20Effect%20Context%20Handle%20Structure))
  ;[1 1 1] Health bar behavior changes for STW
  (|CallParentFunction Damage DamageTags Momentum HitInfo InstigatedBy DamageCauser EffectContext)
  (if (Utilities|Enum|Equal(Enum) (Fortnite|GetSubGame) "Campaign")
    (Utilities|Casting|CastToFortPlayerPawn InstigatedBy
      (:then
        (Variables|Default|SetNPCStatusWidgetShouldDisplayDamage true))
      (:CastFailed
        (Variables|Default|SetNPCStatusWidgetShouldDisplayDamage false)))))

(event EventNPCUpdateHealthBarComponentVisibility
  ;[1 1 1] Health bar behavior changes for STW
  (if (Utilities|Enum|Equal(Enum) (Fortnite|GetSubGame) "Campaign")
    (if (not (Utilities|IsValid (Convert|GetConverterPawn (Variables|Default|GetConvertComponent)) :!class /Script/Engine.KismetSystemLibrary))
      (if (Variables|Default|GetNPCStatusWidgetShouldDisplayDamage)
        _))
    (|CallParentFunction)
    (else
      _)))

(event EventOnPawnHealthChanged_Event_0
  ;[1 1 1] Health bar behavior changes for STW
  (if (Utilities|Enum|Equal(Enum) (Fortnite|GetSubGame) "Campaign")
    (if (not (Utilities|IsValid (Convert|GetConverterPawn (Variables|Default|GetConvertComponent)) :!class /Script/Engine.KismetSystemLibrary))
      (if (Variables|Default|GetNPCStatusWidgetShouldDisplayDamage)
        _))
    (|CallParentFunction)
    (else
      _)))

(fn ApplyVariation
  (Default|SetColorVariation))

(fn AttachAndApplyVocalFX
  (params (SoundLibraryResult :type Sound%20Library%20Play%20Result%20Structure%20%28by%20ref%29))
  (bind (_eventname _bwasblocked _bsuccess _audiocomponents) (Utilities|Struct|BreakSoundLibraryPlayResult SoundLibraryResult))
  (bind _vocalsourceeffectchainoverride (Variables|Default|GetVocalSourceEffectChainOverride)
    (:then
      (if (GameplayTags|MatchesTag _eventname "(TagName=\"SoundLibrary.Irwin.VO\")")
        (for _array_element _audiocomponents
          (Utilities|IsValid _array_element
            (:Is%20Valid
              (Variables|/Script/Engine.AudioComponent|SetSourceEffectChain _vocalsourceeffectchainoverride _array_element)
              (Audio|Components|Audio|Play _array_element 0.0)
              (Transformation|AttachComponentToComponent _array_element (Variables|Transformation|GetRootComponent) "None" "SnapToTarget" "SnapToTarget" "SnapToTarget" true))))))
    (:else)))

(fn BlueprintCanInteract
  (params (InteractingPawn :type Fort%20Pawn%20Object%20Reference) (InteractionBeingAttempted :type EInteractionBeingAttempted%20Enum :default "FirstInteraction"))
  (bind _returnvalue (Actor|GetComponentbyClass "/Script/FortniteGame.FortPawnComponent_Convert"))
  ;[1 1 1] Disable stealing mounts - only let current owner ride already tamed pets
  (Utilities|IsValid _returnvalue
    (:Is%20Valid
      (if (Convert|GetOnlyLeaderInteraction _returnvalue)
        ;;[1 1 1] Enable teammates to revive pets
        (if (Variables|Health|GetbIsDBNO)
          (if (not (GameplayTags|HasMatchingGameplayTag InteractingPawn "(TagName=\"Granted.Riding.Player.IsRiding\")"))
            (return (|CallParentFunction InteractingPawn InteractionBeingAttempted))
            (else
              _))
          (elif (or (not (Utilities|IsValid (Convert|GetConverterPawn _returnvalue) :!class /Script/Engine.KismetSystemLibrary)) (== (Convert|GetConverterPawn _returnvalue) InteractingPawn))
            _))
        (return false)))
    (:Is%20Not%20Valid
      _)))

(fn BlueprintModifyIncomingDamage
  (params (Damage :type Float%20%28single-precision%29) (DamageTags :type Gameplay%20Tag%20Container%20Structure%20%28by%20ref%29) (EffectContext :type Gameplay%20Effect%20Context%20Handle%20Structure%20%28by%20ref%29) (EventInstigator :type Controller%20Object%20Reference) (DamageCauser :type Actor%20Object%20Reference))
  (bind _returnvalue (Utilities|GetFrameCount))
  (bind _returnvalue_2 (|CallParentFunction Damage DamageTags EffectContext EventInstigator DamageCauser))
  (if (Variables|Damage|GetOverrideIncomingDmg)
    (LocalVariables|SetNewDamage 0.0)
    (if (not (Math|Boolean|ANDBoolean (== (LocalVariables|GetLastDamageInstigator) EventInstigator) (== DamageCauser (LocalVariables|GetLast%20DamageCauser)) (== _returnvalue (LocalVariables|GetLastFrameCount))))
      (LocalVariables|SetNewDamage (Variables|Damage|GetOverridenIncomingDmgAmount)))
    (LocalVariables|SetLastDamageInstigator EventInstigator)
    (LocalVariables|SetLast%20DamageCauser DamageCauser)
    (LocalVariables|SetLastFrameCount _returnvalue)
    (Default|CallOnModifiedDamageReceived)
    (return (LocalVariables|GetNewDamage))
    (else
      (return _returnvalue_2))))

(fn CanBeInteractedWithWhenTamed
  (bind _asbool (CallFunction|NPCGetScalableFloatHotfixValue (Variables|NPCInteraction|GetNPC%20AllowInteractWhenHasLeaderHF)))
  (return (and _asbool (Variables|Interaction|GetbCanBeInteractedWithWhileTamed))))

(fn IsValidAutoFireTarget
  (params (Attacker :type Actor%20Object%20Reference))
  (bind _output_get (LocalVariables|SetTeamAffiliation (Team|GetTeamAffiliationBetweenTeams (Team|GetActorTeam Attacker) (Variables|Default|GetLeaderTeam))))
  (if (Utilities|Enum|Equal(Enum) _output_get "Neutral")
    (return (and (Utilities|Enum|Equal(Enum) (Variables|NPCAlertLevel|GetNPC_AlertLevel) "NewEnumerator2") (< (FortPawn|GetHealth) (FortPawn|GetMaxHealth))))
    (else
      (return (not (Utilities|Enum|Equal(Enum) _output_get "Friendly"))))))

(fn NPCCanBeInteractedWith
  (bind _caninteract (|CallParentFunction))
  (return (and _caninteract (Variables|Interaction|GetbTamingEnabled))))

(fn NPCPickupGrabbed_Destroyed
  (params (ActorGrabbed :type Actor%20Object%20Reference))
  (|CallParentFunction ActorGrabbed)
  (Default|CallOnPickupGrabbedDestroyed self ActorGrabbed))

(fn NPCStatusWidget_UpdateKeepVisible
  (params (ShouldKeepVisible :type Boolean))
  (if (Utilities|Enum|Equal(Enum) (Fortnite|GetSubGame) "Campaign")
    (bind (_asbool _asintrounded _asfloat) (CallFunction|NPCGetScalableFloatHotfixValue (Variables|NPCStatusWidgetHealthBar|GetNPC%20MaxDistanceFromDamagedNPCToDrawHeallthBarHF)))
    (if (Math|Boolean|ANDBoolean (> _asfloat (Class|AbilityFunctions|GetDistanceBetweenActors (Pawn|GetControlledPawn (Fortnite|GetFirstLocalFortPlayerController true)) self)) (not (Variables|Health|GetbIsDying)) (not (Variables|NPCStatusWidgetHealthBar|GetNPC%20StatusWidgetHealthBarDisabled)))
      (bind _npcstatuswidget (Variables|Default|GetNPCStatusWidget)
        (:then
          (bind _asnpcstatus_widget (Utilities|Casting|CastToNPCStatusWidget (UserInterface|GetUserWidgetObject _npcstatuswidget)))
          (Class|NPCStatusWidget|SetKeepVisibilityOn _asnpcstatus_widget ShouldKeepVisible :!class /Game/Athena/HUD/Actors/NPCStatusWidget.NPCStatusWidget_C)
          (if ShouldKeepVisible
            (Class|NPCStatusWidget|EventUpdateStatus _asnpcstatus_widget (/ (FortPawn|GetHealth) (FortPawn|GetMaxHealth)) 0.0 false)))
        (:else)))))

(fn OnLaunchCharacter
  (params (LaunchVelocity :type Vector :default "0, 0, 0") (bXYOverride :type Boolean) (bZOverride :type Boolean))
  (bind _stuckbouncing_minlaunchcount (Variables|StuckBouncing|GetStuckBouncing_MinLaunchCount))
  (bind _stuckbouncing_launchcount (Variables|StuckBouncing|GetStuckBouncing_LaunchCount))
  (bind _returnvalue (>= _stuckbouncing_launchcount _stuckbouncing_minlaunchcount))
  (if (Math|Boolean|ANDBoolean (Math|Vector2D|IsNearlyZero2D (Math|Conversions|ToVector2D(Vector) LaunchVelocity) 0.001) (not (GameplayTags|HasMatchingGameplayTag "(TagName=\"Granted.Riding.Creature.IsBeingRidden\")")) (ScalableFloat|GetValueasBool (Variables|StuckBouncing|GetStuckBouncing_UnstuckEnabled)))
    (Math|Integer|IncrementInt (Variables|StuckBouncing|GetStuckBouncing_LaunchCount)
      (:%20%20
        (if _returnvalue
          ;[1 0 0.2283] Eventually would want to do the unstuck from the BT based on a tag added in the GE but I need to move on to Juno :D so changing the velocity directly here will have to do for now
          (bind _stuckbouncing_ge (Variables|StuckBouncing|GetStuckBouncing_GE)
            (:then
              (bind _returnvalue_2 (GameplayEffects|ApplyGameplayEffectToSelf :self (Variables|GameplayAbilities|GetAIPawnAbilitySystemComponent) :GameplayEffectClass _stuckbouncing_ge :Level 0.0))
              (Variables|StuckBouncing|SetStuckBouncing_GeHandle _returnvalue_2))
            (:else
              _))
          ;;[1 1 1] Apply a velocity in pawn forward vector + rotated randomly to help with cases where we are bouncing toward a wall
          (Variables|StuckBouncing|SetStuckBouncing_LaunchCount 0)
          (return (+ (Math|Vector|RotateVectorAroundAxis (* (Transformation|GetActorForwardVector) (Variables|StuckBouncing|GetStuckBouncing_UnstuckVelocity)) (Math|Random|RandomFloatinRange (* (Variables|StuckBouncing|GetStuckBouncing_UnstuckVelocityMaxZRotationAngle) -1.0) (Variables|StuckBouncing|GetStuckBouncing_UnstuckVelocityMaxZRotationAngle)) (Math|Vector|VectorUp)) LaunchVelocity))
          (else
            _))))
    (elif _returnvalue
      (GameplayEffects|RemoveActiveGameplayEffect (Variables|GameplayAbilities|GetAIPawnAbilitySystemComponent) (Variables|StuckBouncing|GetStuckBouncing_GeHandle) -1)
      (Variables|StuckBouncing|SetStuckBouncing_LaunchCount 0)
      (else
        _)))
  (return LaunchVelocity))

(fn OnRep_LeaderTeam
  (Default|UpdateTamingIndicator))

(fn PlaySoundLib
  (params (Event%20Name :type Gameplay%20Tag%20Structure))
  (bind _returnvalue (WorldSubsystems|GetSoundLibrarySubsystem))
  (bind _returnvalue_2 (SoundLibrary|CreateSoundLibraryContext "/Game/Sounds/SoundLibrary/Contexts/Creatures/B_CreatureSoundLibraryContext.B_CreatureSoundLibraryContext_C" (Utilities|Struct|MakeSimpleSoundSettings Event%20Name true false true true) self Event%20Name true true true))
  (Utilities|IsValid _returnvalue_2
    (:Is%20Valid
      (Utilities|IsValid _returnvalue
        (:Is%20Valid
          (bind _outresults (SoundLibrary|PlaySoundwithContext _returnvalue _returnvalue_2))
          (CallFunction|AttachandApplyVocalFX _outresults))))))

(fn PlayWaterSplashatWaterSurface
  (params (Water%20Body :type Fort%20Water%20Body%20Actor%20Object%20Reference) (GameplayCueTag :type Gameplay%20Tag%20Structure))
  (|CallParentFunction Water%20Body GameplayCueTag)
  (Default|PlaySoundLib "(TagName=\"SoundLibrary.Irwin.Action.Swim.Land\")"))

(fn PlayAdditiveHitReacts
  (params (Hit%20Direction :type Vector%20%28by%20ref%29 :default "0, 0, 0") (Anim%20Montage :type Anim%20Montage%20Object%20Reference))
  (if (GameplayTags|HasTag (GameplayTags|GetOwnedGameplayTags :!class /Script/GameplayTags.GameplayTagAssetInterface) "(TagName=\"Athena.AI.PassivePhase.Phasing\")")
    (|CallParentFunction Hit%20Direction)
    (else
      (|CallParentFunction Hit%20Direction Anim%20Montage)))
  (return))

(fn PlayStateChangeAudio
  (params (Index :type Enum_NPC_AlertLevel%20Enum :default "NewEnumerator4"))
  (bind _returnvalue (Utilities|Select :Index Index :NewEnumerator1 "(TagName=\"SoundLibrary.Irwin.VO.Unaware\")" :NewEnumerator2 "(TagName=\"SoundLibrary.Irwin.VO.Alerted\")" :NewEnumerator3 "(TagName=\"SoundLibrary.Irwin.VO.Aggressive\")"))
  (if (GameplayTags|IsGameplayTagValid _returnvalue)
    (Default|PlaySoundLib self _returnvalue)))

(fn SetColorVariation
  (bind _colorvariation (Variables|Default|GetColorVariation))
  (if (> _colorvariation -1.0)
    (bind _returnvalue (Customization|GetSkeletalMeshforPartType (Variables|Default|GetFortCharacterParts) "Body" :!class /Script/FortniteGame.FortCharacterPartsComponent))
    (Rendering|Material|SetCustomPrimitiveDataFloat _returnvalue 0 _colorvariation)))

(fn SetTamingEnabled
  (params (EnableTaming :type Boolean))
  (bind _returnvalue (Actor|GetComponentbyClass "/Script/FortniteGame.FortPawnComponent_Convert"))
  (Variables|Interaction|SetbTamingEnabled EnableTaming)
  (Utilities|IsValid _returnvalue
    (:Is%20Valid
      (bind _caninteract (CallFunction|NPCCanbeInteractedWith))
      (Convert|SetCanbeConverted _returnvalue _caninteract))))

(fn TryShowHealthBar
  (bind _self self)
  (bind _returnvalue (Actor|GetComponentbyClass _self "/Script/FortniteGame.FortPawnComponent_Convert"))
  (bind _returnvalue_2 (Convert|GetConverterPawn _returnvalue))
  (if (Variables|Default|GetbIsBeingRidden)
    ;[1 1 1] Additionnal Riding Checks
    (Utilities|IsValid _returnvalue
      (:Is%20Valid
        (Utilities|IsValid _returnvalue_2
          (:Is%20Valid
            (if (Pawn|IsLocallyControlled _returnvalue_2)
              _
              (elif (Team|OnLocalPlayersTeam _returnvalue_2)
                (|CallParentFunction)))))))
    (else
      (|CallParentFunction))))

(fn UpdateTamingIndicator
  (bind _convertcomponent (Variables|Default|GetConvertComponent))
  (Networking|SwitchHasAuthority
    (:Remote
      (if (Math|Byte|Equal(Byte) (Team|GetActorTeam (Fortnite|GetFirstLocalFortPlayerController true)) (Variables|Default|GetLeaderTeam))
        (Convert|ShowIndicatorMarker _convertcomponent)
        (MiniMap|ShowMapMarker)
        (CallFunction|NPCStatusWidgetUpdateKeepVisible true)
        (else
          (Convert|HideIndicatorMarker _convertcomponent)
          (MiniMap|HideMapMarker)
          (CallFunction|NPCStatusWidgetUpdateKeepVisible false))))))

(fn ConstructionScript
  (|CallParentFunction)
  (Customization|BindEventtoOnFinishedCustomization (Variables|Default|GetFortCharacterParts) (EventDispatchers|CreateEvent :!event On%20Customization%20Applied)))
